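# Niagara effect for the SouthBulk AI's charge loop: a red-hot energy cone with glow, sparks, wind and ground dust.
# System variables for charge delay, duration, cone position and ground offset drive most emitters' timing and placement.
# ConeWind layers core, dark and glow colors across mesh and sprite renderers, alongside DecalGlow, curl-noise GPU sparks, wind meshes and streaks.
# The comet flame emitters and a second cone are disabled.
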NiagaraSystem /SouthBulkAI/Assets/VFX/Charge/NS_SouthBulk_ChargeLoop
dependencies {
 usedDataInterfaces [
  &NiagaraDataInterfaceCurve
  &NiagaraDataInterfaceVectorCurve
  &NiagaraDataInterfaceColorCurve
  &NiagaraDataInterfaceCamera
  &NiagaraDataInterfaceSpriteRendererInfo
  &NiagaraDataInterfaceSkeletalMesh
 ]
 usedDynamicInputs []
 usedModules [
  &self:SystemSpawnScript.SystemScriptSource.SystemScriptGraph.NiagaraNodeAssignment_1.SetVariables_#1
  &self:SystemSpawnScript.SystemScriptSource.SystemScriptGraph.NiagaraNodeAssignment_0.SetVariables_#2
  &self:SystemSpawnScript.SystemScriptSource.SystemScriptGraph.NiagaraNodeAssignment_3.SetVariables_#3
  &self:SystemSpawnScript.SystemScriptSource.SystemScriptGraph.NiagaraNodeAssignment_2.SetVariables_#4
  &self:SystemSpawnScript.SystemScriptSource.SystemScriptGraph.NiagaraNodeAssignment_4.SetVariables_#5
  &SystemState
 ]
}
emitters {
 Comet_Flame_Mesh {
  bEnabled=false
  emitterUpdateScript [
   &EmitterState(
    Life Cycle Mode=::NewEnumerator1
    Loop Behavior=::NewEnumerator1
    Loop Delay=&Add_Float(A=$System.ChargeDelay, B=0)
    Loop Duration=0.5
    UseLoopDelay=true
   )
   &SpawnBurst_Instantaneous(Use Spawn Probability=false, UseLoopCountLimit=true)
  ]
  particleSpawnScript [
   &InitializeParticle(
    Color=<rgba 1 0.1 0.1229 1>
    Lifetime=$System.ChargeDuration
    Mass Mode=::NewEnumerator1
    Mesh Scale Mode=::NewEnumerator4
    Position Offset=<-100 0 0>
    UsePositionOffset=true
    Write Material Random=true
   )
   &InitialMeshOrientation(Mesh Axis To Orient=<-1 0 0>, Mesh Orientation Mode=::NewEnumerator2, Use Rotation Vector=false)
   &DynamicMaterialParameters(
    Index 0 Param 1=2
    Index 0 Param 2=2
    Index 0 Param 3=1
    Index 0 Param 4=1
    Module.Index0Float0Write=true
    Module.Index0Float1Write=true
    Module.Index0Float2Write=true
    Module.Index0Float3Write=true
   )
  ]
  particleUpdateScript [
   &ParticleState()
   /SetVariables(
    Particles.CometLength:NiagaraFloat=&FloatFromCurve(
     FloatCurve=DataInterface() {propertyValues {Curve {keys [
         {arriveTangent=5.772, interpMode=RCIM_Cubic, leaveTangent=5.772, tangentMode=RCTM_User, time=0, value=0}
         {interpMode=RCIM_Cubic, tangentMode=RCTM_User, time=0.1, value=1}
        ]}}}
     Scale Curve=2
    )
    Particles.CometWidth:NiagaraFloat=&FloatFromCurve(
     FloatCurve=DataInterface() {propertyValues {Curve {keys [
         {arriveTangent=2.4908, interpMode=RCIM_Cubic, leaveTangent=2.4908, tangentMode=RCTM_User, time=0, value=0}
         {interpMode=RCIM_Cubic, tangentMode=RCTM_User, time=0.2, value=1}
        ]}}}
     Scale Curve=3
    )
   ) {isSetParameters=true, module=SetVariables_#6}
   &ScaleMeshSize(Scale Factor=&MakeVector(X=$Particles.CometLength, Y=$Particles.CometWidth, Z=$Particles.CometWidth))
   &ScaleColor(
    Scale Alpha=&FloatFromCurve(
     FloatCurve=DataInterface() {propertyValues {Curve {keys [
         {interpMode=RCIM_Cubic, tangentMode=RCTM_User, time=0, value=0}
         {interpMode=RCIM_Cubic, tangentMode=RCTM_User, time=0.2, value=1}
         {interpMode=RCIM_Cubic, tangentMode=RCTM_User, time=0.5, value=1}
         {interpMode=RCIM_Cubic, tangentMode=RCTM_User, time=1, value=0}
        ]}}}
     Scale Curve=0.1
    )
    ScaleA=true
    ScaleRGB=false
   )
   /SetVariables(
    Particles.ColorCore:LinearColor=&ColorFromCurve(ColorCurve=DataInterface() {propertyValues {
       AlphaCurve {keys [<at 0 0>,<at 0.2879 0.1>,<at 1 0 RCIM_Cubic>]}
       BlueCurve {keys [<at 0.1843 0.2 RCIM_Cubic>,<at 0.9816 0.0217>]}
       GreenCurve {keys [<at 0.1843 0.35 RCIM_Cubic>,<at 0.9816 0>]}
       RedCurve {keys [<at 0.1843 2 RCIM_Cubic>,<at 0.9816 1.3>]}
      }})
    Particles.ScaleCore:Vector3f=&Multiply_Vector(A=$Particles.Scale, B=<1 0.8 0.8>)
   ) {isSetParameters=true}
  ]
  rendererClasses [&NiagaraMeshRendererProperties]
  renderers [
   {rendererClass=&NiagaraMeshRendererProperties, rendererIndex=0}
   {rendererClass=&NiagaraMeshRendererProperties, rendererIndex=1}
   {rendererClass=&NiagaraMeshRendererProperties, rendererIndex=2}
  ]
 }
 Comet_Flame_Mesh001 {
  bEnabled=false
  emitterUpdateScript [
   &EmitterState(Life Cycle Mode=::NewEnumerator1, Loop Behavior=::NewEnumerator1, Loop Delay=0.2, Loop Duration=0.5, UseLoopDelay=false)
   &SpawnBurst_Instantaneous(Spawn Count=1, Use Spawn Probability=false)
  ]
  particleSpawnScript [
   &InitializeParticle(
    Color=<rgba 1.5 0.2175 0.15 1>
    Lifetime=$System.ChargeDuration
    Mass Mode=::NewEnumerator1
    Mesh Scale=<-1 1 1>
    Mesh Scale Mode=::NewEnumerator4
    Position Offset=<-30 0 0>
    UsePositionOffset=true
    Write Material Random=true
   ) {module=InitializeParticle001}
   &InitialMeshOrientation(Mesh Orientation Mode=::NewEnumerator2, Orientation Vector=<-1 0 0>, Use Rotation Vector=false)
   &DynamicMaterialParameters(
    Index 0 Param 1=2
    Index 0 Param 2=2
    Index 0 Param 3=1
    Index 0 Param 4=1
    Module.Index0Float0Write=true
    Module.Index0Float1Write=true
    Module.Index0Float2Write=true
    Module.Index0Float3Write=true
   )
  ]
  particleUpdateScript [
   &ParticleState()
   /SetVariables(
    Particles.CometLength:NiagaraFloat=&FloatFromCurve(FloatCurve=DataInterface() {propertyValues {Curve {keys [
         {arriveTangent=5.772, interpMode=RCIM_Cubic, leaveTangent=5.772, tangentMode=RCTM_User, time=0, value=0}
         {interpMode=RCIM_Cubic, tangentMode=RCTM_User, time=0.1, value=1}
        ]}}})
    Particles.CometWidth:NiagaraFloat=&FloatFromCurve(
     FloatCurve=DataInterface() {propertyValues {Curve {keys [
         {arriveTangent=2.4908, interpMode=RCIM_Cubic, leaveTangent=2.4908, tangentMode=RCTM_User, time=0, value=0}
         {interpMode=RCIM_Cubic, tangentMode=RCTM_User, time=0.2, value=1}
         {arriveTangent=-0.7377, interpMode=RCIM_Cubic, leaveTangent=-0.7377, tangentMode=RCTM_User, time=1, value=0.5}
        ]}}}
     Scale Curve=1.66
    )
   ) {isSetParameters=true, module=SetVariables_#6}
   &ScaleMeshSize(Scale Factor=&MakeVector(X=$Particles.CometLength, Y=$Particles.CometWidth, Z=$Particles.CometWidth))
   &ScaleColor(
    Scale Alpha=&FloatFromCurve(FloatCurve=DataInterface() {propertyValues {Curve {keys [<at 0 1>,<at 1 0>]}}})
    Scale RGB=&MakeVectorFromLinearColorRGB(Color=&ColorFromCurve(ColorCurve=DataInterface() {propertyValues {
        AlphaCurve {keys []}
        BlueCurve {keys [<at 0 0.25 RCIM_Cubic>,<at 0.492 0.5>,<at 1 0.875 RCIM_Cubic>]}
        GreenCurve {keys [<at 0 0.4375 RCIM_Cubic>,<at 0.492 0.625>,<at 1 0.9062 RCIM_Cubic>]}
        RedCurve {keys [<at 0 1 RCIM_Cubic>,<at 0.492 1>,<at 1 1 RCIM_Cubic>]}
       }}))
    ScaleA=true
    ScaleRGB=true
   )
   /SetVariables(
    Particles.Color_2:LinearColor=&ColorFromCurve(ColorCurve=DataInterface() {propertyValues {
       AlphaCurve {keys [
         <at 0.0167 0>
         {arriveTangent=0.4279, interpMode=RCIM_Cubic, leaveTangent=0.4279, time=0.1571, value=0.3}
         <at 0.4841 0.2>
         <at 0.9227 0 RCIM_Cubic>
        ]}
       BlueCurve {keys [<at 0.0295 0.125 RCIM_Cubic>,<at 0.8605 1 RCIM_Cubic>]}
       GreenCurve {keys [<at 0.0295 0.2438 RCIM_Cubic>,<at 0.8605 1 RCIM_Cubic>]}
       RedCurve {keys [<at 0.0295 2.5 RCIM_Cubic>,<at 0.8605 1 RCIM_Cubic>]}
      }})
    Particles.Scale_2:Vector3f=&Multiply_Vector(A=$StackContext.Scale, B=<-1 1 1>)
   ) {isSetParameters=true}
  ]
  rendererClasses [&NiagaraMeshRendererProperties]
  renderers [
   {rendererClass=&NiagaraMeshRendererProperties, rendererIndex=0}
   {rendererClass=&NiagaraMeshRendererProperties, rendererIndex=1}
  ]
 }
 ConeWind {
  emitterUpdateScript [
   &EmitterState(
    Life Cycle Mode=::NewEnumerator1
    Loop Behavior=::NewEnumerator1
    Loop Delay=&Add_Float(A=$System.ChargeDelay, B=0)
    Loop Duration Mode=::NewEnumerator1
    UseLoopDelay=true
   )
   &SpawnBurst_Instantaneous(Use Spawn Probability=false, UseLoopCountLimit=true)
  ]
  particleSpawnScript [
   &InitializeParticle(
    Color=<rgba 0.8 0 0.0136 1>
    Lifetime=$System.ChargeDuration
    Mesh Scale=&Multiply_VectorByFloat(Vector=<0.6 1 1>, Float=1.6)
    Mesh Scale Mode=::NewEnumerator4
    Position Offset:NiagaraPosition=$System.EnergyConePosition
    Sprite Size=&Multiply_Vector2D_ByFloat(Float=150, Vector2D=<1 1>)
    Sprite Size Mode=::NewEnumerator4
    UsePositionOffset=true
    Write Material Random=true
   )
   &InitialMeshOrientation(Mesh Orientation Mode=::NewEnumerator2, Rotation=<0 0.5 0>, Use Rotation Vector=false)
   &CameraOffset(Camera Offset Amount=5)
   /SetVariables(Particles.CameraOffsetFlare=15) {isSetParameters=true}
  ]
  particleUpdateScript [
   &ParticleState()
   &ScaleMeshSize(Scale Factor=&Multiply_VectorByFloat(
     Vector=&MakeVector(
      X=&FloatFromCurve(FloatCurve=DataInterface() {propertyValues {Curve {keys [
           {arriveTangent=0.0679, interpMode=RCIM_Cubic, leaveTangent=0.0679, tangentMode=RCTM_User, time=0, value=0.1}
           {interpMode=RCIM_Cubic, tangentMode=RCTM_User, time=1, value=1}
          ]}}})
      Y=1
      Z=1
     )
     Float=&MakeFloatFromVector(
      Vector=&VectorFromFloat(Value=&Lerp_Float(
        A=1
        B=2.5
        Alpha=&RemapRange(
         Input Value To Remap=/Niagara/DynamicInputs/Camera/DistanceToCamera(Use Owner Position:NiagaraBool=true, Camera Query:NiagaraDataInterfaceCamera=DataInterface())
         Input Min=12500
         Input Max=25000
         Output Min=0
         Output Max=1
        )
       ))
      Channel=::NewEnumerator0
     )
    )) {enabled=false}
   &ScaleColor(
    Scale Alpha=&FloatFromCurve(FloatCurve=DataInterface() {propertyValues {Curve {keys [
         {arriveTangent=0.1759, interpMode=RCIM_Cubic, leaveTangent=0.1759, tangentMode=RCTM_User, time=0, value=0}
         {interpMode=RCIM_Cubic, tangentMode=RCTM_User, time=0.2, value=1}
         {arriveTangentWeight=0.1026, interpMode=RCIM_Cubic, leaveTangentWeight=0.1368, tangentMode=RCTM_User, time=0.6, value=1}
         {arriveTangent=-0.4064, interpMode=RCIM_Cubic, leaveTangent=-0.4064, tangentMode=RCTM_User, time=1, value=0}
        ]}}})
    ScaleA=true
    ScaleRGB=false
   )
   /SetVariables(
    Particles.ColorGlow:LinearColor=&ColorFromCurve(ColorCurve=DataInterface() {propertyValues {
       AlphaCurve {keys [<at 0.012 0>,<at 0.0775 1 RCIM_Cubic>,<at 0.8636 0.8>,<at 1 0 RCIM_Cubic>]}
       BlueCurve {keys [<at 0.0829 0.4057>,<at 0.2817 0.1656>,<at 0.8924 0.1052>]}
       GreenCurve {keys [<at 0.0829 0.4>,<at 0.2817 0.1599>,<at 0.8924 0.09>]}
       RedCurve {keys [<at 0.0829 8>,<at 0.2817 3>,<at 0.8924 1>]}
      }})
    Particles.PositionGlow:NiagaraPosition=&AddVectorToPosition(Position=$Particles.Position, Vector=<0 0 0>)
   ) {isSetParameters=true, module=SetVariables_#7}
   /SetVariables(
    Particles.ColorGlowSoft:LinearColor=&MultiplyLinearColor(Linear Color=$Particles.ColorGlow, Scale Factor=<1 0 0 1>)
    Particles.SpriteSizeGlow:Vector2f=&Multiply_Vector2D_ByFloat(Float=0.6, Vector2D=$Particles.SpriteSize)
   ) {isSetParameters=true, module=SetVariables_#8}
   /SetVariables(Particles.SpriteSizeLong:Vector2f=&Multiply_Vector2D(A=$Particles.SpriteSize, B=<2 0.1>)) {isSetParameters=true}
   /SetVariables(Particles.ColorMain:LinearColor=$Particles.Color, Particles.ScaleMain:Vector3f=$Particles.Scale) {isSetParameters=true}
   /SetVariables(
    Particles.ColorCore:LinearColor=&MultiplyLinearColor(
     Linear Color=<rgba 7 1.05 1.05 1>
     Scale Factor=&MakeVector4FromVectorAndFloat(
      "Vector (RGB)"=<1 1 1>
      "Float (Alpha)"=&MakeFloatFromVector4(Vector 4=&MakeVector4_FromLinearColor(LinearColor=$Particles.Color), Channel=::NewEnumerator3)
     )
    )
    Particles.ScaleCore:Vector3f=&Multiply_Vector(A=$Particles.Scale, B=<1 1 1>)
   ) {isSetParameters=true}
   /SetVariables(
    Particles.ColorDark:LinearColor=&MultiplyLinearColor(
     Linear Color=<rgba 0.03 0.009 0.0108 1>
     Scale Factor=&MakeVector4FromVectorAndFloat(
      "Vector (RGB)"=<1 1 1>
      "Float (Alpha)"=&MakeFloatFromVector4(Vector 4=&MakeVector4_FromLinearColor(LinearColor=$Particles.Color), Channel=::NewEnumerator3)
     )
    )
    Particles.ScaleDark:Vector3f=&Multiply_VectorByFloat(Vector=$Particles.Scale, Float=1.2)
   ) {isSetParameters=true}
   /SetVariables(Particles.SpriteSizeDark:Vector2f=&Multiply_Vector2D_ByFloat(Float=1, Vector2D=$Particles.SpriteSize)) {isSetParameters=true, module=SetVariables_#9}
  ]
  rendererClasses [&NiagaraSpriteRendererProperties, &NiagaraMeshRendererProperties]
  renderers [
   {rendererClass=&NiagaraSpriteRendererProperties}
   {rendererClass=&NiagaraSpriteRendererProperties, rendererIndex=1}
   {rendererClass=&NiagaraSpriteRendererProperties, rendererIndex=2}
   {rendererClass=&NiagaraMeshRendererProperties, rendererIndex=3}
   {rendererClass=&NiagaraMeshRendererProperties, rendererIndex=4}
   {rendererClass=&NiagaraMeshRendererProperties, rendererIndex=5}
   {rendererClass=&NiagaraSpriteRendererProperties, rendererIndex=6}
   {rendererClass=&NiagaraMeshRendererProperties, rendererIndex=7}
  ]
 }
 ConeWind001 {
  bEnabled=false
  emitterUpdateScript [
   &EmitterState(Life Cycle Mode=::NewEnumerator1, Loop Behavior=::NewEnumerator1, Loop Delay=0.1, Loop Duration Mode=::NewEnumerator1, UseLoopDelay=true)
   &SpawnBurst_Instantaneous(Use Spawn Probability=false, UseLoopCountLimit=true)
  ]
  particleSpawnScript [
   &InitializeParticle(
    Color=<rgba 8 1.0333 0.4 0.55>
    Lifetime=$System.ChargeDuration
    Mesh Scale=&Multiply_VectorByFloat(Vector=<0.7 1 1>, Float=1.5)
    Mesh Scale Mode=::NewEnumerator4
    Sprite Size Mode=::NewEnumerator3
    Uniform Sprite Size=390.7748
    UsePositionOffset=false
    Write Material Random=true
   )
   &InitialMeshOrientation(Mesh Orientation Mode=::NewEnumerator2, Rotation=<0 0.5 0>, Use Rotation Vector=false)
   &CameraOffset(Camera Offset Amount=10)
  ]
  particleUpdateScript [
   &ParticleState()
   &ScaleMeshSize(Scale Factor=&Multiply_VectorByFloat(
     Vector=&MakeVector(
      X=&FloatFromCurve(FloatCurve=DataInterface() {propertyValues {Curve {keys [
           {arriveTangent=0.0679, interpMode=RCIM_Cubic, leaveTangent=0.0679, tangentMode=RCTM_User, time=0, value=0.1}
           {interpMode=RCIM_Cubic, tangentMode=RCTM_User, time=1, value=1}
          ]}}})
      Y=1
      Z=1
     )
     Float=&MakeFloatFromVector(
      Vector=&VectorFromFloat(Value=&Lerp_Float(
        A=1
        B=2.5
        Alpha=&RemapRange(
         Input Value To Remap=&DistanceToCamera(Use Owner Position:NiagaraBool=true, Camera Query:NiagaraDataInterfaceCamera=DataInterface())
         Input Min=12500
         Input Max=25000
         Output Min=0
         Output Max=1
        )
       ))
      Channel=::NewEnumerator0
     )
    )) {enabled=false}
   &ScaleColor(
    Scale Alpha=&FloatFromCurve(
     FloatCurve=DataInterface() {propertyValues {Curve {keys [{interpMode=RCIM_Cubic, tangentMode=RCTM_User, time=0, value=0}, {interpMode=RCIM_Cubic, tangentMode=RCTM_User, time=0.5, value=1}]}}}
     CurveIndex=$Engine.System.Age
    )
    ScaleA=true
    ScaleRGB=true
   )
   &ScaleColor(
    Scale Alpha=&FloatFromCurve(FloatCurve=DataInterface() {propertyValues {Curve {keys [{interpMode=RCIM_Cubic, tangentMode=RCTM_User, time=0.4, value=1}, {interpMode=RCIM_Cubic, tangentMode=RCTM_User, time=1, value=0}]}}})
    ScaleA=true
    ScaleRGB=true
   ) {module=ScaleColor001}
   /SetVariables(Particles.ColorDark:LinearColor=&MultiplyLinearColor(Linear Color=$Particles.Color, Scale Factor=<0.03 0.02 0 2.5>)) {isSetParameters=true}
   /SetVariables(Particles.ScaleDark:Vector3f=&Multiply_VectorByFloat(Vector=$Particles.Scale, Float=1.3)) {isSetParameters=true}
   /SetVariables(
    Particles.ColorGlow:LinearColor=&ColorFromCurve(ColorCurve=DataInterface() {propertyValues {
       AlphaCurve {keys [<at 0.012 0>,<at 0.1348 1 RCIM_Cubic>,<at 0.5636 0.8>,<at 1 0 RCIM_Cubic>]}
       BlueCurve {keys [<at 0.0837 0.24 RCIM_Cubic>,<at 0.9068 0.04>]}
       GreenCurve {keys [<at 0.0837 1.2 RCIM_Cubic>,<at 0.9068 0.072>]}
       RedCurve {keys [<at 0.0837 6 RCIM_Cubic>,<at 0.9068 1>]}
      }})
    Particles.PositionGlow:NiagaraPosition=&AddVectorToPosition(Position=$Particles.Position, Vector=<-10 0 0>)
   ) {isSetParameters=true, module=SetVariables_#7}
   /SetVariables(
    Particles.ColorGlowSoft:LinearColor=&MultiplyLinearColor(Linear Color=$Particles.ColorGlow, Scale Factor=<1 0.66 0.2 0.5>)
    Particles.SpriteSizeGlow:Vector2f=&Multiply_Vector2D_ByFloat(Float=1, Vector2D=$Particles.SpriteSize)
   ) {isSetParameters=true, module=SetVariables_#8}
   /SetVariables(Particles.SpriteSizeDark:Vector2f=&Multiply_Vector2D_ByFloat(Float=1, Vector2D=$Particles.SpriteSize)) {isSetParameters=true, module=SetVariables_#9}
  ]
  rendererClasses [&NiagaraSpriteRendererProperties, &NiagaraMeshRendererProperties]
  renderers [
   {rendererClass=&NiagaraSpriteRendererProperties}
   {rendererClass=&NiagaraSpriteRendererProperties, rendererIndex=1}
   {rendererClass=&NiagaraSpriteRendererProperties, rendererIndex=2}
   {rendererClass=&NiagaraMeshRendererProperties, rendererIndex=3}
   {rendererClass=&NiagaraMeshRendererProperties, rendererIndex=4}
  ]
 }
 DecalGlow {
  emitterUpdateScript [
   &EmitterState(Life Cycle Mode=::NewEnumerator1, Loop Behavior=::NewEnumerator1, Loop Duration=1, UseLoopDelay=false)
   &SpawnBurst_Instantaneous(Use Spawn Probability=false, UseLoopCountLimit=true)
  ]
  particleSpawnScript [
   &InitializeParticle(
    Lifetime=&Add_Float(A=$System.ChargeDuration, B=0.5)
    Position Offset=&AddVectorToPosition(Position=$System.EnergyConePosition, Vector=<-150 0 0>)
    UsePositionOffset=true
    Write Material Random=true
   )
   &PartitionParticles(ID to Partition:NiagaraInt32=&ReturnExecIndex()) {enabled=false}
   &CameraOffset(Camera Offset Amount=40)
   /SetVariables(Particles.CameraOffsetFront=55) {enabled=false, isSetParameters=true}
  ]
  particleUpdateScript [
   &SkeletalMeshLocation(
    Bone Sampling Mode=::NewEnumerator4
    MaskSpawnGroup=false
    Skeletal Mesh=DataInterface() {propertyValues {
      DefaultMesh=null
      ExcludeBoneName=root
      OptionalComponentClass=null
      PreviewMesh=/SouthBulkAI/Assets/Meshes/M_LRG_NarrowFlea_ZenithBoss_NPC
      SkinningMode=SkinOnTheFly
      SoftSourceActor=null
      SourceMode=Default
      UvSetIndex=0
      WholeMeshLOD=-1
      bExcludeBone=false
      bReadDeformedGeometry=true
      bRequireCurrentFrameData=true
      ComponentTags []
      FilteredBones [pelvis]
      FilteredSockets []
      MeshUserParameter {parameter {name=None, typeDefHandle {registeredTypeIndex=1}, varData []}}
      SamplingRegions []
     }}
   ) {enabled=false}
   &ParticleState()
   &FN_Sprite_Attributes(
    Color=&ColorFromCurve(ColorCurve=DataInterface() {propertyValues {
       AlphaCurve {keys [<at 0 0>,<at 0.458 1 RCIM_Cubic>,<at 0.79 0.7>,<at 1 0 RCIM_Cubic>]}
       BlueCurve {keys [<at 0.6665 0.05 RCIM_Cubic>]}
       GreenCurve {keys [<at 0.6665 0 RCIM_Cubic>]}
       RedCurve {keys [<at 0.6665 3 RCIM_Cubic>]}
      }})
    Create Color Attribute=true
    Create Dynamic Material Parameter Attributes=true
    Create SpriteSize Attribute=true
    Index 0 Param 1:NiagaraFloat=&FloatFromCurve(
     FloatCurve=DataInterface() {propertyValues {Curve {keys [
         {arriveTangent=0.2363, interpMode=RCIM_Cubic, leaveTangent=0.2363, tangentMode=RCTM_User, time=0, value=0.1}
         {arriveTangent=2.4473, interpMode=RCIM_Cubic, leaveTangent=2.4473, tangentMode=RCTM_User, time=0.5, value=1}
        ]}}}
     Scale Curve=2
    )
    Module.Index0Float0Write=true
    Module.Index0Float1Write=false
    Module.Index0Float2Write=true
    Module.Index0Float3Write=false
    SpriteSize=&Multiply_Vector2D_ByFloat(Float=360, Vector2D=<1 1>)
   )
  ]
  rendererClasses [&NiagaraSpriteRendererProperties]
  renderers [
   {rendererClass=&NiagaraSpriteRendererProperties}
   {rendererClass=&NiagaraSpriteRendererProperties, rendererIndex=1}
   {rendererClass=&NiagaraSpriteRendererProperties, rendererIndex=2}
  ]
 }
 Dust {
  emitterUpdateScript [
   &EmitterState(
    Life Cycle Mode=::NewEnumerator1
    Loop Behavior=::NewEnumerator1
    Loop Delay=&Add_Float(A=$System.ChargeDelay, B=0)
    Loop Duration=&Add_Float(A=$System.ChargeDuration, B=-0.3)
    UseLoopDelay=true
   )
   &SpawnRate(SpawnRate=20, Use Spawn Probability=false)
  ]
  particleSpawnScript [
   &InitializeParticle(
    Color Maximum=<rgba 0.8385 0.9347 1 0.8>
    Color Minimum=<rgba 0.9 0.9055 1 0.6>
    Color Mode=::NewEnumerator2
    Lifetime Max=1.5
    Lifetime Min=0.75
    Lifetime Mode=::NewEnumerator1
    Position Offset:NiagaraPosition=$System.DustGroundOffset
    Sprite Size Mode=::NewEnumerator3
    Uniform Sprite Size=&Multiply_Float(A=&RandomRangeFloat(Minimum=200, Maximum=300), B=2)
    UsePositionOffset=true
    Write Material Random=true
   ) {module=InitializeParticle001}
   &ShapeLocation(
    MaskSpawnGroup=false
    Non Uniform Scale=<1 1 0.5>
    Override Seed=false
    Sphere Radius=20
    Sphere Surface Distribution=0.8095
    UseOwnerScale=true
   )
   &AddVelocityFromPoint(
    Module.Use Falloff Distance=false
    Origin Offset=&AddVectorToPosition(Position=$System.DustGroundOffset, Vector=<400 0 -100>)
    Velocity Strength=&Multiply_Float(A=&RandomRangeFloat(Minimum=1000, Maximum=1600), B=1)
   )
  ]
  particleUpdateScript [
   &ParticleState()
   &Drag(Drag=&RandomRangeFloat(Minimum=1, Maximum=2), Use Linear Drag=true, Use Rotational Drag=false)
   &SpriteRotationRate(Rotation Rate=&Multiply_Float(
     A=&RandomRangeFloat(Minimum=-3, Maximum=3)
     B=&FloatFromCurve(
      FloatCurve=DataInterface() {propertyValues {Curve {keys [
          {arriveTangent=-4.0107, interpMode=RCIM_Cubic, leaveTangent=-4.0107, tangentMode=RCTM_User, time=-0.0046, value=1.0113}
          {arriveTangent=-0.2496, interpMode=RCIM_Cubic, leaveTangent=-0.2496, tangentMode=RCTM_User, time=0.5, value=0.16}
         ]}}}
      CurveIndex=$Emitter.Age
      Scale Curve=150
     )
    )) {enabled=false}
   &ScaleSpriteSize(Scale Factor=&Vector2DFromFloat(Value=&FloatFromCurve(FloatCurve=DataInterface() {propertyValues {Curve {keys [
          {arriveTangent=2.3678, interpMode=RCIM_Cubic, leaveTangent=2.3678, tangentMode=RCTM_User, time=0, value=0.2}
          {arriveTangent=0.7401, interpMode=RCIM_Cubic, leaveTangent=0.7401, tangentMode=RCTM_User, time=0.3, value=0.7}
          {arriveTangent=0.2539, interpMode=RCIM_Cubic, leaveTangent=0.2539, tangentMode=RCTM_User, time=0.9978, value=1}
         ]}}})))
   &SolveForcesAndVelocity(Clamp Velocity=false, Limit Acceleration=false)
   &ScaleColor(
    Scale Alpha=&FloatFromCurve(
     FloatCurve=DataInterface() {propertyValues {Curve {keys [
         {arriveTangent=0.4611, interpMode=RCIM_Cubic, leaveTangent=0.4611, tangentMode=RCTM_User, time=0, value=0}
         {arriveTangent=-2.5865, interpMode=RCIM_Cubic, leaveTangent=-2.5865, tangentMode=RCTM_User, time=0.2, value=1}
         {arriveTangent=-0.4609, interpMode=RCIM_Cubic, leaveTangent=-0.4609, tangentMode=RCTM_User, time=1, value=0}
        ]}}}
     Scale Curve=&RandomRangeFloat(Minimum=0.2, Maximum=0.3)
    )
    Scale RGB=&MakeVectorFromLinearColorRGB(Color=&ColorFromCurve(ColorCurve=DataInterface() {propertyValues {
        AlphaCurve {keys []}
        BlueCurve {keys [<at 0.0393 2.4 RCIM_Cubic>,<at 0.1425 0.1729>,<at 0.4414 0.08>]}
        GreenCurve {keys [<at 0.0393 2.7 RCIM_Cubic>,<at 0.1425 0.3797>,<at 0.4414 0.0883>]}
        RedCurve {keys [<at 0.0393 6 RCIM_Cubic>,<at 0.1425 1>,<at 0.4414 0.1>]}
       }}))
    ScaleA=true
    ScaleRGB=true
   )
   &DynamicMaterialParameters(
    Index 0 Param 1=&Add_Float(
     A=$Particles.RandomFloat
     B=&FloatFromCurve(
      FloatCurve=DataInterface() {propertyValues {Curve {keys [
          {arriveTangent=2.7533, interpMode=RCIM_Cubic, leaveTangent=1.8958, tangentMode=RCTM_User, time=0, value=1}
          {arriveTangent=0.8292, interpMode=RCIM_Cubic, leaveTangent=0.8292, tangentMode=RCTM_User, time=0.321, value=1.3453}
          {arriveTangent=0.0529, interpMode=RCIM_Cubic, leaveTangent=0.2021, tangentMode=RCTM_User, time=1, value=1.5}
         ]}}}
      CurveIndex=&Add_Float(A=&Multiply_Float(A=$Particles.Initial.MaterialRandom, B=0.5), B=$Particles.NormalizedAge)
     )
    )
    Index 0 Param 2=&FloatFromCurve(FloatCurve=DataInterface() {propertyValues {Curve {keys [
         {arriveTangent=0.0453, interpMode=RCIM_Cubic, tangentMode=RCTM_Break, time=0, value=-1}
         {arriveTangent=0.0604, interpMode=RCIM_Cubic, leaveTangent=0.0604, tangentMode=RCTM_User, time=1, value=-1}
        ]}}})
    Index 0 Param 3=&FloatFromCurve(FloatCurve=DataInterface() {propertyValues {Curve {keys [
         {arriveTangent=-53.9199, interpMode=RCIM_Cubic, leaveTangent=-8.9727, tangentMode=RCTM_User, time=0, value=1}
         {arriveTangent=0.2347, interpMode=RCIM_Cubic, leaveTangent=0.2723, tangentMode=RCTM_Break, time=0.25, value=0}
        ]}}})
    Index 0 Param 4=&FloatFromCurve(FloatCurve=DataInterface() {propertyValues {Curve {keys [<at 0.5 1>,<at 1 0>]}}})
    Index 2 Param 1=&FloatFromCurve(FloatCurve=DataInterface() {propertyValues {Curve {keys [
         {interpMode=RCIM_Cubic, leaveTangent=4.2485, tangentMode=RCTM_Break, time=0, value=0}
         {arriveTangent=1.0548, interpMode=RCIM_Cubic, leaveTangent=1.0548, tangentMode=RCTM_User, time=0.17, value=0.5715}
         <at 1 1>
        ]}}})
    Index 2 Param 2=&FloatFromCurve(
     FloatCurve=DataInterface() {propertyValues {Curve {keys [
         {arriveTangent=0.9316, interpMode=RCIM_Cubic, leaveTangent=0.9316, tangentMode=RCTM_User, time=0, value=0.1}
         {arriveTangent=0.0259, interpMode=RCIM_Cubic, leaveTangent=0.0259, tangentMode=RCTM_User, time=1, value=0.4}
        ]}}}
     Scale Curve=&RandomRangeFloat(Minimum=0.2, Maximum=0.7)
    )
    Index 2 Param 3=&FloatFromCurve(FloatCurve=DataInterface() {propertyValues {Curve {keys [
         {interpMode=RCIM_Cubic, leaveTangent=2.3064, tangentMode=RCTM_Break, time=0, value=1}
         {interpMode=RCIM_Cubic, leaveTangent=11.1233, tangentMode=RCTM_Break, time=1, value=2}
        ]}}})
    Index 2 Param 4=&FloatFromCurve(
     FloatCurve=DataInterface() {propertyValues {Curve {keys [
         {arriveTangent=-0.6867, interpMode=RCIM_Cubic, leaveTangent=-0.6867, tangentMode=RCTM_User, time=0, value=1}
         {interpMode=RCIM_Cubic, leaveTangent=-1.2956, tangentMode=RCTM_Break, time=1, value=0.75}
        ]}}}
     Scale Curve=1.6
    )
    Module.Param0WriteEnabled=true
    Module.Param0WriteEnabled2=true
    Module.Param1WriteEnabled=false
    Module.Param1WriteEnabled2=true
    Module.Param2WriteEnabled=true
    Module.Param2WriteEnabled2=true
    Module.Param3WriteEnabled=true
    Module.Param3WriteEnabled2=true
    Write Parameter Index 2=true
   )
   &CameraOffset(Camera Offset Amount=/Niagara/DynamicInputs/Math/CalculateParticleRadius(
     Radius Calculation Type:ENiagaraCollisionRadiusOptions=::NewEnumerator0
     Particle Radius:NiagaraFloat=1
     Mesh Dimensions:Vector3f=<1 1 1>
     Method for Calculating Particle Radius:ENiagaraCalculateRadiusOptions=::NewEnumerator2
    ))
  ]
  rendererClasses [&NiagaraSpriteRendererProperties]
  renderers [
   {rendererClass=&NiagaraSpriteRendererProperties}
   {rendererClass=&NiagaraSpriteRendererProperties, rendererIndex=1}
  ]
 }
 Sparks003 {
  simTarget=GPUComputeSim
  emitterUpdateScript [
   &EmitterState(
    Life Cycle Mode=::NewEnumerator1
    Loop Behavior=::NewEnumerator1
    Loop Delay=$System.ChargeDelay
    Loop Duration=$System.ChargeDuration
    UseLoopDelay=true
   )
   &SpawnRate(SpawnRate=100, Use Spawn Probability=false)
  ]
  particleSpawnScript [
   &InitializeParticle(
    Color=<rgba 2 2 2 1>
    Lifetime Max=0.8
    Lifetime Min=0.3
    Lifetime Mode=::NewEnumerator1
    Position Offset:NiagaraPosition=$System.EnergyConePosition
    Sprite Size Mode=::NewEnumerator3
    Uniform Sprite Size=&RandomRangeFloat(Minimum=6, Maximum=8)
    UsePositionOffset=true
    Write Material Random=true
   )
   &CameraOffset(Camera Offset Amount=25)
   &ShapeLocation(
    Hemisphere Distribution=<1 0.5>
    MaskSpawnGroup=false
    Offset=<-300 0 -50>
    Offset Coordinate Space=::Local
    Offset Mode=::NewEnumerator0
    Override Seed=false
    Sphere Radius=220
    Sphere Surface Distribution=0.85
    UseOwnerScale=true
   )
   &DynamicMaterialParameters(
    Index 0 Param 1=2
    Index 0 Param 2=10
    Index 0 Param 3=1
    Module.Index0Float0Write=true
    Module.Index0Float1Write=true
    Module.Index0Float2Write=true
    Module.Index0Float3Write=false
   )
   &AddVelocity(
    Module.Use Falloff Distance=false
    Override Seed=false
    Velocity Mode=::NewEnumerator1
    Velocity Speed=&RandomRangeFloat(Minimum=400, Maximum=900)
   )
  ]
  particleUpdateScript [
   &ParticleState()
   &AccelerationForce(Acceleration=<0 0 -500>) {enabled=false}
   &CurlNoiseForce(
    Bias Noise=false
    Noise Frequency=400
    Noise Strength=&FloatFromCurve(
     FloatCurve=DataInterface() {propertyValues {Curve {keys [{interpMode=RCIM_Cubic, tangentMode=RCTM_User, time=0, value=0}, {interpMode=RCIM_Cubic, tangentMode=RCTM_User, time=0.3, value=1}]}}}
     Scale Curve=5000
    )
    Pan Noise=false
   )
   &Drag(Drag=&RandomRangeFloat(Minimum=2, Maximum=3), Use Linear Drag=true, Use Rotational Drag=false)
   &AccelerationForce(Acceleration=<0 0 100>) {module=AccelerationForce001}
   &SolveForcesAndVelocity(Clamp Velocity=false, Limit Acceleration=false)
   &ScaleSpriteSizeBySpeed(Max Scale Factor=<1 4>)
   &ScaleSpriteSize(Scale Factor=&Vector2DFromFloat(Value=&FloatFromCurve(
      FloatCurve=DataInterface() {propertyValues {Curve {keys [
          {arriveTangent=7.1771, interpMode=RCIM_Cubic, leaveTangent=7.1771, tangentMode=RCTM_User, time=0, value=0}
          {arriveTangent=0.1164, interpMode=RCIM_Cubic, leaveTangent=0.1164, tangentMode=RCTM_User, time=0.15, value=1}
          {arriveTangent=-0.2479, interpMode=RCIM_Cubic, leaveTangent=-0.2479, tangentMode=RCTM_User, time=0.5, value=1}
          <at 1 0 RCIM_Cubic>
         ]}}}
      Scale Curve=&RandomRangeFloat(Minimum=0.6, Maximum=1.2)
     )))
   &ScaleColor(
    Scale RGB=&VectorFromFloat(Value=&FloatFromCurve(
      FloatCurve=DataInterface() {propertyValues {Curve {keys [
          {interpMode=RCIM_Cubic, tangentMode=RCTM_User, time=0, value=1}
          <at 0.3 0.2>
          {interpMode=RCIM_Cubic, tangentMode=RCTM_User, time=1, value=0}
         ]}}}
      Scale Curve=10
     ))
    ScaleA=true
    ScaleRGB=true
   ) {enabled=false}
   &ScaleColor(
    Scale Alpha=&FloatFromCurve(FloatCurve=DataInterface() {propertyValues {Curve {keys [
         <at 0 0 RCIM_Cubic>
         {arriveTangent=0.159, interpMode=RCIM_Cubic, leaveTangent=0.159, tangentMode=RCTM_User, time=0.1, value=1}
         {interpMode=RCIM_Cubic, tangentMode=RCTM_Break, time=0.85, value=1}
         {interpMode=RCIM_Cubic, tangentMode=RCTM_Break, time=1, value=0}
        ]}}})
    Scale RGB=&MakeVectorFromLinearColorRGB(Color=&ColorFromCurve(ColorCurve=DataInterface() {propertyValues {
        AlphaCurve {keys [<at 0 1 RCIM_Cubic>]}
        BlueCurve {keys [<at 0 0.7 RCIM_Cubic>,<at 0.2653 0.1>,<at 0.9955 0>]}
        GreenCurve {keys [<at 0 0.91 RCIM_Cubic>,<at 0.2653 0.1>,<at 0.9955 0>]}
        RedCurve {keys [<at 0 7 RCIM_Cubic>,<at 0.2653 2>,<at 0.9955 1>]}
       }}))
    ScaleA=false
    ScaleRGB=true
   ) {module=ScaleColor001}
   &FN_Sprite_Attributes(
    Color=&ColorFromCurve(ColorCurve=DataInterface() {propertyValues {
       AlphaCurve {keys [<at 0.0151 0>,<at 0.0946 0.075 RCIM_Cubic>,<at 0.7029 0 RCIM_Cubic>]}
       BlueCurve {keys [<at 0.0598 0 RCIM_Cubic>,<at 0.7123 0.0333>]}
       GreenCurve {keys [<at 0.0598 0.0667 RCIM_Cubic>,<at 0.7123 0>]}
       RedCurve {keys [<at 0.0598 1 RCIM_Cubic>,<at 0.7123 1>]}
      }})
    Create Color Attribute=true
    Create Dynamic Material Parameter Attributes=true
    Create SpriteSize Attribute=true
    Index 0 Param 1=1
    Index 0 Param 2=10
    Index 0 Param 3=1
    Module Output:ENiagara_AttributeSamplingApplyOutput=::NewEnumerator1
    Module.Index0Float0Write=true
    Module.Index0Float1Write=true
    Module.Index0Float2Write=true
    Module.Index0Float3Write=true
    SpriteSize=&Multiply_Vector2D(A=$StackContext.SpriteSize, B=<7 2.6>)
   )
  ]
  rendererClasses [&NiagaraSpriteRendererProperties]
  renderers [
   {rendererClass=&NiagaraSpriteRendererProperties}
   {rendererClass=&NiagaraSpriteRendererProperties, rendererIndex=1}
  ]
 }
 Wind {
  emitterUpdateScript [
   &EmitterState(
    Enable Distance Culling=true
    Life Cycle Mode=::NewEnumerator1
    Loop Behavior=::NewEnumerator1
    Loop Delay=&Add_Float(A=$System.ChargeDelay, B=0)
    Loop Duration=&Add_Float(A=$System.ChargeDuration, B=-0.3)
    MaxDistance=6000
    Module.bUseMaxDistance=true
    Module.bUseMinDistance=false
    Scalability Mode=::NewEnumerator1
    UseLoopDelay=true
   )
   &SpawnRate(SpawnRate=30, Use Spawn Probability=false)
   &SpawnBurst_Instantaneous(Spawn Count=12, Use Spawn Probability=false, UseLoopCountLimit=false) {enabled=false}
  ]
  particleSpawnScript [
   /SetVariables(Particles.WindScale:NiagaraFloat=&RandomRangeFloat(Minimum=0, Maximum=1)) {enabled=false, isSetParameters=true}
   &InitializeParticle(
    Lifetime=0.5
    Mesh Scale=&Lerp_Vector(A=<1.5 1.5 2>, B=<2 2 4>, Alpha=&RandomRangeFloat(Minimum=0, Maximum=1))
    Mesh Scale Mode=::NewEnumerator4
    Position Offset:NiagaraPosition=$System.EnergyConePosition
    UsePositionOffset=true
    Write Material Random=true
   )
   &InitialMeshOrientation(
    Mesh Axis To Orient=<0 0 1>
    Mesh Orientation Mode=::NewEnumerator2
    Rotation=&RandomRangeVector(Minimum=<0 0 0>, Maximum=<0 0 1>)
    Use Rotation Vector=true
   )
   &AddVelocity(Velocity=<-1000 0 0>, Velocity Speed Scale=&RandomRangeFloat(Minimum=1.2, Maximum=1.6))
  ]
  particleUpdateScript [
   &ParticleState()
   /SetVariables(Particles.Position=&Add_Vector(
     A=Linked() {linkedVariable {name=Particles.Initial.Position, type=NiagaraPosition}}
     B=&Multiply_VectorByFloat(
      Vector=&ReturnMeshOrientationAxis(Mesh Orientation Axis=::ZAxis, Mesh Orientation Quat=$Particles.MeshOrientation)
      Float=&FloatFromCurve(
       FloatCurve=DataInterface() {propertyValues {Curve {keys [{arriveTangent=-2.1833, interpMode=RCIM_Cubic, leaveTangent=-2.1833, tangentMode=RCTM_User, time=0, value=0}, <at 1 -1>]}}}
       Scale Curve=&Lerp_Float(A=100, B=300, Alpha=$Particles.WindScale)
      )
     )
    )) {enabled=false, isSetParameters=true}
   &UpdateMeshOrientation(
    Rotation Coordinate Space=::NewEnumerator3
    Rotation Rate=&FloatFromCurve(FloatCurve=DataInterface() {propertyValues {Curve {keys [
         {arriveTangent=-2.0408, interpMode=RCIM_Cubic, time=0, value=1}
         {arriveTangent=-2.0603, interpMode=RCIM_Cubic, leaveTangent=-11.7273, tangentMode=RCTM_User, time=1, value=0.05}
        ]}}})
    Rotation Vector=<0 0 1>
   ) {enabled=false, module=UpdateMeshOrientation001}
   &ScaleMeshSize(Scale Factor=&VectorFromCurve(
     VectorCurve=DataInterface() {propertyValues {
       XCurve {keys [
         {arriveTangent=1.9177, interpMode=RCIM_Cubic, leaveTangent=1.9177, tangentMode=RCTM_User, time=0, value=0.15}
         {arriveTangent=0.2969, interpMode=RCIM_Cubic, leaveTangent=0.8713, tangentMode=RCTM_User, time=1, value=1}
        ]}
       YCurve {keys [
         {arriveTangent=1.9177, interpMode=RCIM_Cubic, leaveTangent=1.9177, tangentMode=RCTM_User, time=0, value=0.15}
         {arriveTangent=0.2969, interpMode=RCIM_Cubic, leaveTangent=0.5259, tangentMode=RCTM_User, time=1, value=1}
        ]}
       ZCurve {keys [
         {arriveTangent=4.3595, interpMode=RCIM_Cubic, leaveTangent=4.3595, tangentMode=RCTM_User, time=0, value=0.5}
         <at 0.2 1 RCIM_Cubic>
         {arriveTangent=-1.6862, interpMode=RCIM_Cubic, leaveTangent=-1.6862, tangentMode=RCTM_User, time=1, value=0.1}
        ]}
      }}
     Scale Curve=<2 2 2>
    ))
   &CameraOffset(Camera Offset Amount=&Lerp_Float(A=70, B=140, Alpha=$Particles.NormalizedAge))
   &ScaleColor(
    Scale Alpha=&FloatFromCurve(
     FloatCurve=DataInterface() {propertyValues {Curve {keys [
         {arriveTangent=8.125, interpMode=RCIM_Cubic, time=0, value=0}
         {interpMode=RCIM_Cubic, tangentMode=RCTM_User, time=0.2, value=1}
         {arriveTangent=-0.8126, interpMode=RCIM_Cubic, leaveTangent=-0.8126, tangentMode=RCTM_User, time=0.7791, value=0.7}
         {interpMode=RCIM_Cubic, leaveTangent=-2.3395, time=1, value=0}
        ]}}}
     Scale Curve=0.4
    )
    Scale RGB=&MakeVectorFromLinearColorRGB(Color=&ColorFromCurve(ColorCurve=DataInterface() {propertyValues {
        AlphaCurve {keys []}
        BlueCurve {keys [<at 0.1072 0.1963 RCIM_Cubic>,<at 0.6909 0.21 RCIM_Cubic>]}
        GreenCurve {keys [<at 0.1072 0.046 RCIM_Cubic>,<at 0.6909 0.2475 RCIM_Cubic>]}
        RedCurve {keys [<at 0.1072 2.3 RCIM_Cubic>,<at 0.6909 0.3 RCIM_Cubic>]}
       }}))
    ScaleA=true
    ScaleRGB=true
   )
   &DynamicMaterialParameters(
    Index 0 Param 1=&FloatFromCurve(FloatCurve=DataInterface() {propertyValues {Curve {keys [
         {arriveTangent=-13.4039, interpMode=RCIM_Cubic, leaveTangent=-13.4039, tangentMode=RCTM_User, time=0, value=1}
         <at 0.1 0>
         {arriveTangent=1.1111, interpMode=RCIM_Cubic, leaveTangent=1.1111, time=0.4, value=0}
         {interpMode=RCIM_Cubic, leaveTangent=16.4602, time=1, value=1}
        ]}}})
    Index 0 Param 2=&FloatFromCurve(
     FloatCurve=DataInterface() {propertyValues {Curve {keys [
         {arriveTangent=2.4843, interpMode=RCIM_Cubic, leaveTangent=2.4843, tangentMode=RCTM_User, time=0, value=0}
         {interpMode=RCIM_Cubic, leaveTangent=-0.8461, time=1, value=1}
        ]}}}
     Scale Curve=0.5
    )
    Index 0 Param 4=&FloatFromCurve(
     FloatCurve=DataInterface() {propertyValues {Curve {keys [
         {arriveTangent=0.2805, interpMode=RCIM_Cubic, time=0, value=1}
         {arriveTangent=2.4494, interpMode=RCIM_Cubic, leaveTangent=2.4494, tangentMode=RCTM_User, time=1, value=2}
        ]}}}
     Scale Curve=&RandomRangeFloat(Minimum=0.6, Maximum=0.8)
    )
    Module.Param0WriteEnabled=true
    Module.Param1WriteEnabled=true
    Module.Param2WriteEnabled=false
    Module.Param3WriteEnabled=true
   )
   &Drag(Drag=0.5, Use Linear Drag=true, Use Rotational Drag=true)
   &SolveForcesAndVelocity(Clamp Velocity=false, Limit Acceleration=false)
  ]
  rendererClasses [&NiagaraMeshRendererProperties]
  renderers [{rendererClass=&NiagaraMeshRendererProperties, rendererIndex=0}]
 }
 WindStreaks {
  emitterUpdateScript [
   &EmitterState(
    Life Cycle Mode=::NewEnumerator1
    Loop Behavior=::NewEnumerator1
    Loop Delay=&Add_Float(A=$System.ChargeDelay, B=0)
    Loop Duration=&Add_Float(A=$System.ChargeDuration, B=-0.3)
    UseLoopDelay=true
   )
   &SpawnRate(SpawnRate=80, Use Spawn Probability=false)
  ]
  particleSpawnScript [
   &InitializeParticle(
    Color Maximum=<rgba 0.3 0.3 0.3 0.6>
    Color Minimum=<rgba 0.35 0.35 0.35 0.7>
    Color Mode=::NewEnumerator2
    Lifetime=&RandomRangeFloat(Minimum=0.3, Maximum=0.5)
    Sprite Size=&Multiply_Vector2D_ByFloat(
     Float=&RandomRangeFloat(Minimum=140, Maximum=170)
     Vector2D=&RandomRangeVector2D(Minimum=<0.5 2>, Maximum=<1 3>)
    )
    Sprite Size Mode=::NewEnumerator4
    UsePositionOffset=false
    Write Material Random=true
   )
   &PartitionParticles()
   &ShapeLocation(
    MaskSpawnGroup=false
    Non Uniform Scale=<3 1 0.5>
    Offset=&RotateVectorByQuaternion(
     Vector To Rotate=&AddVectorToPosition(
      Position=$System.DustGroundOffset
      Vector=&MergeHelper_VectorSwitch(0=<-50 -100 0>, 1=<-50 100 0>, 2=?, 3=?, Index=$Particles.PartitionParticles.Partition, Num=::NewEnumerator0)
     )
     Quaternion=$Engine.Owner.Rotation
    )
    Offset Mode=::NewEnumerator0
    Override Seed=false
    Sphere Radius=50
    UseOwnerScale=true
   )
   &AddVelocity(
    Module.Use Falloff Distance=false
    Origin Offset=&AddVectorToPosition(Position=$System.DustGroundOffset, Vector=<150 0 -80>)
    Override Seed=false
    Velocity Mode=::NewEnumerator1
    Velocity Speed=&RandomRangeFloat(Minimum=900, Maximum=1600)
   )
   &SubUVAnimation(
    "Sprite Renderer "=DataInterface() {propertyValues {SpriteRenderer=&self:WindStreaks_0.NiagaraSpriteRendererProperties_1}}
    SubUV Animation Mode:ENiagara_SubUVLookupModeV2=::NewEnumerator2
    UseEndFrame=false
    UseStartFrame=false
   )
  ]
  particleUpdateScript [
   &ParticleState()
   &Drag(Drag=2, Use Linear Drag=true, Use Rotational Drag=true)
   &InheritSourceMovement(
    Applied Position Scale=&VectorFromFloat(Value=&FloatFromCurve(FloatCurve=DataInterface() {propertyValues {Curve {keys [<at 0 1>,<at 0.5 0>]}}}, Scale Curve=0.4))
    ApplyForce=false
    ApplyPosition=true
    ApplyPositionFalloff=false
   )
   &SolveForcesAndVelocity(Clamp Velocity=false, Limit Acceleration=false)
   &ScaleColor(
    Scale Alpha=&FloatFromCurve(FloatCurve=DataInterface() {propertyValues {Curve {keys [<at 0 0>,<at 0.1 1>,<at 1 0>]}}}, Scale Curve=2)
    ScaleA=true
    ScaleRGB=true
   )
   &ScaleSpriteSize(
    Scale Factor=&MakeVector2D(
     X=&FloatFromCurve(FloatCurve=DataInterface() {propertyValues {Curve {keys [<at 0 0.4>,<at 0.4 1>]}}})
     Y=&FloatFromCurve(FloatCurve=DataInterface() {propertyValues {Curve {keys [
          {arriveTangent=0.2172, interpMode=RCIM_Cubic, leaveTangent=0.2172, tangentMode=RCTM_User, time=0, value=0.8}
          {interpMode=RCIM_Cubic, tangentMode=RCTM_User, time=1, value=1}
         ]}}})
    )
    Scale Sprite Size Mode=::NewEnumerator2
   )
   &DynamicMaterialParameters(
    Index 0 Param 1=20
    Index 0 Param 2=1
    Index 0 Param 3=1
    Index 0 Param 4=&FloatFromCurve(FloatCurve=DataInterface() {propertyValues {Curve {keys [
         {interpMode=RCIM_Cubic, tangentMode=RCTM_User, time=0, value=0}
         {arriveTangent=2.7922, interpMode=RCIM_Cubic, leaveTangent=2.7922, tangentMode=RCTM_User, time=1, value=1}
        ]}}})
    Index 2 Param 1=2
    Index 2 Param 2=&FloatFromCurve(FloatCurve=DataInterface() {propertyValues {Curve {keys [
         {arriveTangent=0.0308, interpMode=RCIM_Cubic, leaveTangent=0.0308, tangentMode=RCTM_User, time=0, value=1}
         {arriveTangent=-2.7663, interpMode=RCIM_Cubic, leaveTangent=-2.7663, tangentMode=RCTM_User, time=1, value=0}
        ]}}})
    Module.Param0WriteEnabled=true
    Module.Param0WriteEnabled2=true
    Module.Param1WriteEnabled=true
    Module.Param1WriteEnabled2=true
    Module.Param2WriteEnabled=true
    Module.Param2WriteEnabled2=true
    Module.Param3WriteEnabled=true
    Module.Param3WriteEnabled2=true
    Write Parameter Index 2=true
   )
  ]
  rendererClasses [&NiagaraSpriteRendererProperties]
  renderers [{rendererClass=&NiagaraSpriteRendererProperties}]
 }
}
summary {
 systemName=NS_SouthBulk_ChargeLoop
 emitters [
  {emitterName=Wind, rendererClasses [&NiagaraMeshRendererProperties]}
  {emitterName=Dust, rendererClasses [&NiagaraSpriteRendererProperties]}
  {emitterName=ConeWind, rendererClasses [&NiagaraSpriteRendererProperties, &NiagaraMeshRendererProperties]}
  {emitterName=WindStreaks, rendererClasses [&NiagaraSpriteRendererProperties]}
  {bEnabled=false, emitterName=Comet_Flame_Mesh, rendererClasses [&NiagaraMeshRendererProperties]}
  {
   bEnabled=false
   emitterName=ConeWind001
   rendererClasses [&NiagaraSpriteRendererProperties, &NiagaraMeshRendererProperties]
  }
  {bEnabled=false, emitterName=Comet_Flame_Mesh001, rendererClasses [&NiagaraMeshRendererProperties]}
  {emitterName=Sparks003, simTarget=GPUComputeSim, rendererClasses [&NiagaraSpriteRendererProperties]}
  {emitterName=DecalGlow, rendererClasses [&NiagaraSpriteRendererProperties]}
 ]
 userVariables [User.Flame_Size {type=NiagaraFloat, defaultValue=1, description=""}]
}
