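# Niagara effect for the start of the SouthBulk AI's charge, combining wind and an energy cone.
# EnergyCone bursts a layered mesh, while delayed, distance-culled Wind meshes and ramping WindStreaks sprites from two ground points add motion.
# System-level wind direction and ground-offset variables place and orient these layers.
# Dust, Dust001, DirtChunks and Minimal are disabled, some timed to ChargeUpTime, and DustCircle has no renderers.
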NiagaraSystem /SouthBulkAI/Assets/VFX/Charge/NS_SouthBulk_ChargeStart
dependencies {
 usedDataInterfaces [
  &NiagaraDataInterfaceCurve
  &NiagaraDataInterfaceColorCurve
  &NiagaraDataInterfaceSpriteRendererInfo
  &NiagaraDataInterfaceCamera
 ]
 usedDynamicInputs []
 usedModules [
  &self:SystemSpawnScript.SystemScriptSource.SystemScriptGraph.NiagaraNodeAssignment_3.SetVariables_#1
  &self:SystemSpawnScript.SystemScriptSource.SystemScriptGraph.NiagaraNodeAssignment_1.SetVariables_#2
  &self:SystemSpawnScript.SystemScriptSource.SystemScriptGraph.NiagaraNodeAssignment_2.SetVariables_#3
  &SystemState
 ]
}
emitters {
 DirtChunks {
  bEnabled=false
  emitterUpdateScript [
   &EmitterState(
    Life Cycle Mode=::NewEnumerator1
    Loop Behavior=::NewEnumerator1
    Loop Delay=&Add_Float(A=$System.ChargeUpTime, B=-0.2)
    Loop Duration=0.5
    UseLoopDelay=true
   )
   &SpawnRate(
    Spawn Probability=&MakeCustomFloatFromBool(Bool=$User.IsSliding, True Float=1, False Float=0)
    SpawnRate=&FloatFromCurve(
     FloatCurve=DataInterface() {propertyValues {Curve {
        defaultValue=1
        keys [{interpMode=RCIM_Cubic, tangentMode=RCTM_User, time=0, value=0}, {interpMode=RCIM_Cubic, tangentMode=RCTM_User, time=1, value=1}]
       }}}
     CurveIndex=$Engine.System.Age
     Scale Curve=12
    )
    Use Spawn Probability=true
   )
  ]
  particleSpawnScript [
   /SetVariables(Particles.VisibilityTag=&SpawnGroup()) {isSetParameters=true}
   &InitializeParticle(
    Color=<rgba 0.6 0.4 0.2 0.78>
    Lifetime=&UniformRangedFloat(Minimum=0.5, Maximum=0.75)
    Sprite Rotation Mode=::NewEnumerator2
    Sprite Size Mode=::NewEnumerator1
    Uniform Sprite Size Max=75
    Uniform Sprite Size Min=50
    UsePositionOffset=false
    Write Material Random=true
   ) {module=InitializeParticle001}
   &PartitionParticles()
   &SkeletalMeshLocation(
    Bone / Socket Index=$Particles.PartitionParticles.Partition
    Bone Sampling Mode=::NewEnumerator4
    MaskSpawnGroup=false
    Orientation Sampling=::NewEnumerator1
    Sampled Position Offset=&MergeHelper_VectorSwitch(0=<0 0 0>, 1=<-25 -5 0>, 2=?, 3=?, Index=$Particles.PartitionParticles.Partition, Num=::NewEnumerator0)
    Skeletal Mesh=$System.Skeletal Mesh
   )
   &AddVelocity(
    Cone Angle=20
    Cone Axis=<-1 0 0>
    Distribution Along Cone Axis=0.3
    Module.UseConeFalloff=true
    Module.UseVelDistribution=true
    Override Seed=false
    Speed Falloff From Cone Axis=0.5
    Velocity Mode=::NewEnumerator2
    Velocity Speed=&RandomRangeFloat(Minimum=600, Maximum=1000)
   )
  ]
  particleUpdateScript [
   &ParticleState()
   &GravityForce(Gravity=<0 0 -800>) {enabled=false}
   &Drag(Drag=&RandomRangeFloat(Minimum=2.5, Maximum=10), Use Linear Drag=true, Use Rotational Drag=false)
   &ScaleColor(
    Scale Alpha=&FloatFromCurve(
     FloatCurve=DataInterface() {propertyValues {Curve {keys [
         {arriveTangent=0.0639, interpMode=RCIM_Cubic, leaveTangent=0.0639, tangentMode=RCTM_User, time=0, value=0}
         {arriveTangent=0.0047, interpMode=RCIM_Cubic, leaveTangent=0.0047, tangentMode=RCTM_User, time=0.1408, value=0.9847}
         {arriveTangent=-1.146, interpMode=RCIM_Cubic, leaveTangent=-1.146, time=0.8, value=0.8}
         {arriveTangent=-0.0405, interpMode=RCIM_Cubic, leaveTangent=-0.0405, tangentMode=RCTM_User, time=1, value=0}
        ]}}}
     Scale Curve=&RandomRangeFloat(Minimum=1.1, Maximum=1)
    )
    Scale RGB=<0.3 0.3 0.3>
    ScaleA=true
    ScaleRGB=true
   )
   /SetVariables(Particles.ColorDust:LinearColor=&MultiplyLinearColor(
     Linear Color=<rgba 0.75 0.3929 0.3087 5>
     Scale Factor=&MakeVector4FromVectorAndFloat(
      "Vector (RGB)"=<0.3 0.3 0.3>
      "Float (Alpha)"=&FloatFromCurve(
       FloatCurve=DataInterface() {propertyValues {Curve {keys [
           {arriveTangent=0.0639, interpMode=RCIM_Cubic, leaveTangent=0.0639, tangentMode=RCTM_User, time=0, value=0}
           {arriveTangent=0.0047, interpMode=RCIM_Cubic, leaveTangent=0.0047, tangentMode=RCTM_User, time=0.1408, value=0.9847}
           {arriveTangent=-1.146, interpMode=RCIM_Cubic, leaveTangent=-1.146, time=0.8, value=0.8}
           {arriveTangent=-0.0405, interpMode=RCIM_Cubic, leaveTangent=-0.0405, tangentMode=RCTM_User, time=1, value=0}
          ]}}}
       Scale Curve=0.85
      )
     )
    )) {isSetParameters=true}
   &InheritSourceMovement(Applied Position Scale=<0.85 0.85 0.85>, ApplyForce=false, ApplyPosition=true, ApplyPositionFalloff=false)
   &SolveForcesAndVelocity(Clamp Velocity=false, Limit Acceleration=false)
   &ScaleSpriteSize(Scale Factor=&Vector2DFromFloat(Value=&FloatFromCurve(
      FloatCurve=DataInterface() {propertyValues {Curve {keys [
          {arriveTangent=1.5713, interpMode=RCIM_Cubic, leaveTangent=1.5713, tangentMode=RCTM_User, time=0, value=0.25}
          {arriveTangent=0.0879, interpMode=RCIM_Cubic, leaveTangent=0.0879, tangentMode=RCTM_User, time=1, value=0.8}
         ]}}}
      Scale Curve=2.8
     )))
  ]
  rendererClasses [&NiagaraSpriteRendererProperties]
  renderers [
   {rendererClass=&NiagaraSpriteRendererProperties}
   {rendererClass=&NiagaraSpriteRendererProperties, rendererIndex=1}
  ]
 }
 Dust {
  bEnabled=false
  emitterUpdateScript [
   &EmitterState(Life Cycle Mode=::NewEnumerator1, Loop Behavior=::NewEnumerator1, Loop Duration=1, UseLoopDelay=true)
   &SpawnBurst_Instantaneous(Spawn Count=11, Use Spawn Probability=false)
  ]
  particleSpawnScript [
   &FN_SetSpawnGroup(Normalize Spawn Group=true, Number of Spawn Groups=1, Particles Per Group=10)
   &InitializeParticle(
    Color Maximum=<rgba 0.7 0.7 0.7 1>
    Color Minimum=<rgba 0.4 0.4 0.4 0.6>
    Color Mode=::NewEnumerator2
    Lifetime Max=1
    Lifetime Min=0.6
    Lifetime Mode=::NewEnumerator1
    Sprite Size Mode=::NewEnumerator1
    Uniform Sprite Size Max=200
    Uniform Sprite Size Min=100
    UsePositionOffset=false
    Write Material Random=true
   ) {module=InitializeParticle001}
   &TorusLocation(
    Align Torus With Axis=true
    Handle Radius=5
    Large Radius=10
    MaskSpawnGroup=false
    Torus Distribution Mode:ENiagaraTorusDistributionMode=::NewEnumerator1
    U Position:NiagaraFloat=&Lerp_Float(A=0.05, B=0.95, Alpha=$Particles.Normalized Spawn Group)
   )
   &AddVelocityFromPoint(
    Module.Use Falloff Distance=false
    Origin Offset=<10 0 0>
    Velocity Strength=&RandomRangeFloat(Minimum=800, Maximum=1200)
   )
  ]
  particleUpdateScript [
   &ParticleState()
   &Drag(Drag=&RandomRangeFloat(Minimum=4, Maximum=5), Use Linear Drag=true, Use Rotational Drag=false)
   &SpriteRotationRate(Rotation Rate=&Multiply_Float(
     A=&RandomRangeFloat(Minimum=-3, Maximum=3)
     B=&FloatFromCurve(
      FloatCurve=DataInterface() {propertyValues {Curve {keys [
          {arriveTangent=-4.0107, interpMode=RCIM_Cubic, leaveTangent=-4.0107, tangentMode=RCTM_User, time=-0.0046, value=1.0113}
          {arriveTangent=-0.2496, interpMode=RCIM_Cubic, leaveTangent=-0.2496, tangentMode=RCTM_User, time=0.5, value=0.16}
         ]}}}
      CurveIndex=$Emitter.Age
      Scale Curve=150
     )
    ))
   &WindForce(Air Resistance=0.4, Wind Speed=<0 0 1>, Wind Speed Scale=&RandomRangeFloat(Minimum=400, Maximum=900))
   &ScaleSpriteSize(Scale Factor=&Vector2DFromFloat(Value=&FloatFromCurve(FloatCurve=DataInterface() {propertyValues {Curve {keys [
          {arriveTangent=2.3678, interpMode=RCIM_Cubic, leaveTangent=2.3678, tangentMode=RCTM_User, time=0, value=0.2}
          {arriveTangent=0.7401, interpMode=RCIM_Cubic, leaveTangent=0.7401, tangentMode=RCTM_User, time=0.3, value=0.7}
          {arriveTangent=0.2539, interpMode=RCIM_Cubic, leaveTangent=0.2539, tangentMode=RCTM_User, time=0.9978, value=1}
         ]}}})))
   &SolveForcesAndVelocity(Clamp Velocity=false, Limit Acceleration=false)
   &ScaleColor(
    Scale Alpha=&FloatFromCurve(
     FloatCurve=DataInterface() {propertyValues {Curve {keys [
         {arriveTangent=5.4511, interpMode=RCIM_Cubic, leaveTangent=5.4511, tangentMode=RCTM_User, time=0, value=0.4}
         {arriveTangent=-1.4813, interpMode=RCIM_Cubic, leaveTangent=-1.4813, tangentMode=RCTM_User, time=0.2, value=1}
         {arriveTangent=-0.484, interpMode=RCIM_Cubic, leaveTangent=-0.484, tangentMode=RCTM_User, time=1, value=0}
        ]}}}
     Scale Curve=&RandomRangeFloat(Minimum=0.5, Maximum=0.3)
    )
    Scale RGB=&VectorFromFloat(Value=1)
    ScaleA=true
    ScaleRGB=true
   )
   &DynamicMaterialParameters(
    Index 0 Param 1=&Add_Float(
     A=$Particles.RandomFloat
     B=&FloatFromCurve(
      FloatCurve=DataInterface() {propertyValues {Curve {keys [
          {arriveTangent=2.7533, interpMode=RCIM_Cubic, leaveTangent=1.8958, tangentMode=RCTM_User, time=0, value=1}
          {arriveTangent=0.8292, interpMode=RCIM_Cubic, leaveTangent=0.8292, tangentMode=RCTM_User, time=0.321, value=1.3453}
          {arriveTangent=0.0529, interpMode=RCIM_Cubic, leaveTangent=0.2021, tangentMode=RCTM_User, time=1, value=1.5}
         ]}}}
      CurveIndex=&Add_Float(A=&Multiply_Float(A=$Particles.Initial.MaterialRandom, B=0.5), B=$Particles.NormalizedAge)
     )
    )
    Index 0 Param 2=&FloatFromCurve(FloatCurve=DataInterface() {propertyValues {Curve {keys [
         {arriveTangent=0.0453, interpMode=RCIM_Cubic, tangentMode=RCTM_Break, time=0, value=-1}
         {arriveTangent=0.0604, interpMode=RCIM_Cubic, leaveTangent=0.0604, tangentMode=RCTM_User, time=1, value=-1}
        ]}}})
    Index 0 Param 3=&FloatFromCurve(FloatCurve=DataInterface() {propertyValues {Curve {keys [
         {arriveTangent=-53.9199, interpMode=RCIM_Cubic, leaveTangent=-8.9727, tangentMode=RCTM_User, time=0, value=1}
         {arriveTangent=0.2347, interpMode=RCIM_Cubic, leaveTangent=0.2723, tangentMode=RCTM_Break, time=0.25, value=0}
        ]}}})
    Index 0 Param 4=&FloatFromCurve(FloatCurve=DataInterface() {propertyValues {Curve {keys [<at 0.5 1>,<at 1 0>]}}})
    Index 2 Param 1=&FloatFromCurve(FloatCurve=DataInterface() {propertyValues {Curve {keys [
         {interpMode=RCIM_Cubic, leaveTangent=4.2485, tangentMode=RCTM_Break, time=0, value=0}
         {arriveTangent=1.0548, interpMode=RCIM_Cubic, leaveTangent=1.0548, tangentMode=RCTM_User, time=0.17, value=0.5715}
         <at 1 1>
        ]}}})
    Index 2 Param 2=&FloatFromCurve(
     FloatCurve=DataInterface() {propertyValues {Curve {keys [
         {arriveTangent=0.9316, interpMode=RCIM_Cubic, leaveTangent=0.9316, tangentMode=RCTM_User, time=0, value=0.1}
         {arriveTangent=0.0259, interpMode=RCIM_Cubic, leaveTangent=0.0259, tangentMode=RCTM_User, time=1, value=0.4}
        ]}}}
     Scale Curve=&RandomRangeFloat(Minimum=0.2, Maximum=0.7)
    )
    Index 2 Param 3=&FloatFromCurve(FloatCurve=DataInterface() {propertyValues {Curve {keys [
         {interpMode=RCIM_Cubic, leaveTangent=2.3064, tangentMode=RCTM_Break, time=0, value=1}
         {interpMode=RCIM_Cubic, leaveTangent=11.1233, tangentMode=RCTM_Break, time=1, value=2}
        ]}}})
    Index 2 Param 4=&FloatFromCurve(
     FloatCurve=DataInterface() {propertyValues {Curve {keys [
         {arriveTangent=-0.6867, interpMode=RCIM_Cubic, leaveTangent=-0.6867, tangentMode=RCTM_User, time=0, value=1}
         {interpMode=RCIM_Cubic, leaveTangent=-1.2956, tangentMode=RCTM_Break, time=1, value=0.75}
        ]}}}
     Scale Curve=1.6
    )
    Module.Param0WriteEnabled=true
    Module.Param0WriteEnabled2=true
    Module.Param1WriteEnabled=false
    Module.Param1WriteEnabled2=true
    Module.Param2WriteEnabled=true
    Module.Param2WriteEnabled2=true
    Module.Param3WriteEnabled=true
    Module.Param3WriteEnabled2=true
    Write Parameter Index 2=true
   )
   &CameraOffset(Camera Offset Amount=/Niagara/DynamicInputs/Math/CalculateParticleRadius(
     Radius Calculation Type:ENiagaraCollisionRadiusOptions=::NewEnumerator0
     Particle Radius:NiagaraFloat=1
     Mesh Dimensions:Vector3f=<1 1 1>
     Method for Calculating Particle Radius:ENiagaraCalculateRadiusOptions=::NewEnumerator2
    ))
  ]
  rendererClasses [&NiagaraSpriteRendererProperties]
  renderers [
   {rendererClass=&NiagaraSpriteRendererProperties}
   {rendererClass=&NiagaraSpriteRendererProperties, rendererIndex=1}
  ]
 }
 Dust001 {
  bEnabled=false
  emitterUpdateScript [
   &EmitterState(
    Life Cycle Mode=::NewEnumerator1
    Loop Behavior=::NewEnumerator1
    Loop Delay=&Add_Float(A=$System.ChargeUpTime, B=0)
    Loop Duration=1
    UseLoopDelay=true
   )
   &SpawnBurst_Instantaneous(Spawn Count=11, Use Spawn Probability=false)
  ]
  particleSpawnScript [
   &FN_SetSpawnGroup(Normalize Spawn Group=true, Number of Spawn Groups=1, Particles Per Group=10)
   &InitializeParticle(
    Color Maximum=<rgba 0.8385 0.9347 1 0.8>
    Color Minimum=<rgba 0.9 0.9055 1 0.6>
    Color Mode=::NewEnumerator2
    Lifetime Max=1.2
    Lifetime Min=0.75
    Lifetime Mode=::NewEnumerator1
    Sprite Size Mode=::NewEnumerator3
    Uniform Sprite Size=&Multiply_Float(
     A=&RandomRangeFloat(Minimum=100, Maximum=200)
     B=&FloatFromCurve(
      FloatCurve=DataInterface() {propertyValues {Curve {keys [
          {interpMode=RCIM_Cubic, tangentMode=RCTM_User, time=0, value=0.5}
          {arriveTangentWeight=0.15, interpMode=RCIM_Cubic, leaveTangentWeight=0.0333, tangentMode=RCTM_User, time=0.5, value=1}
          {tangentMode=RCTM_User, time=1, value=0.5}
         ]}}}
      CurveIndex=$Particles.Normalized Spawn Group
     )
    )
    UsePositionOffset=false
    Write Material Random=true
   ) {module=InitializeParticle001}
   &TorusLocation(
    Align Torus With Axis=true
    Handle Radius=5
    Large Radius=15
    MaskSpawnGroup=false
    Torus Distribution Mode:ENiagaraTorusDistributionMode=::NewEnumerator1
    U Position:NiagaraFloat=&Lerp_Float(A=-0.1, B=0.1, Alpha=$Particles.Normalized Spawn Group)
   )
   &AddVelocityFromPoint(
    Module.Use Falloff Distance=false
    Origin Offset=<-7 0 0>
    Velocity Strength=&Multiply_Float(
     A=&RandomRangeFloat(Minimum=1000, Maximum=1600)
     B=&FloatFromCurve(
      FloatCurve=DataInterface() {propertyValues {Curve {keys [
          {arriveTangent=1.2775, interpMode=RCIM_Cubic, leaveTangent=1.2775, tangentMode=RCTM_User, time=0, value=0.4}
          {arriveTangent=1.5825, interpMode=RCIM_Cubic, leaveTangent=1.5825, tangentMode=RCTM_User, time=0.37, value=0.95}
          {arriveTangentWeight=0.045, interpMode=RCIM_Cubic, leaveTangentWeight=0.2027, tangentMode=RCTM_User, time=0.5, value=1.04}
          {arriveTangent=-1.4249, interpMode=RCIM_Cubic, leaveTangent=-1.4249, tangentMode=RCTM_User, time=0.63, value=0.95}
          {arriveTangent=-1.4935, interpMode=RCIM_Cubic, leaveTangent=-1.4935, tangentMode=RCTM_User, time=1, value=0.4}
         ]}}}
      CurveIndex=$Particles.Normalized Spawn Group
      Scale Curve=-1
     )
    )
   )
  ]
  particleUpdateScript [
   &ParticleState()
   &Drag(Drag=&RandomRangeFloat(Minimum=4, Maximum=6), Use Linear Drag=true, Use Rotational Drag=false)
   &SpriteRotationRate(Rotation Rate=&Multiply_Float(
     A=&RandomRangeFloat(Minimum=-3, Maximum=3)
     B=&FloatFromCurve(
      FloatCurve=DataInterface() {propertyValues {Curve {keys [
          {arriveTangent=-4.0107, interpMode=RCIM_Cubic, leaveTangent=-4.0107, tangentMode=RCTM_User, time=-0.0046, value=1.0113}
          {arriveTangent=-0.2496, interpMode=RCIM_Cubic, leaveTangent=-0.2496, tangentMode=RCTM_User, time=0.5, value=0.16}
         ]}}}
      CurveIndex=$Emitter.Age
      Scale Curve=150
     )
    ))
   &WindForce(Air Resistance=0.4, Wind Speed=<0 0 1>, Wind Speed Scale=&RandomRangeFloat(Minimum=400, Maximum=900)) {enabled=false}
   &ScaleSpriteSize(Scale Factor=&Vector2DFromFloat(Value=&FloatFromCurve(FloatCurve=DataInterface() {propertyValues {Curve {keys [
          {arriveTangent=2.3678, interpMode=RCIM_Cubic, leaveTangent=2.3678, tangentMode=RCTM_User, time=0, value=0.2}
          {arriveTangent=0.7401, interpMode=RCIM_Cubic, leaveTangent=0.7401, tangentMode=RCTM_User, time=0.3, value=0.7}
          {arriveTangent=0.2539, interpMode=RCIM_Cubic, leaveTangent=0.2539, tangentMode=RCTM_User, time=0.9978, value=1}
         ]}}})))
   &SolveForcesAndVelocity(Clamp Velocity=false, Limit Acceleration=false)
   &ScaleColor(
    Scale Alpha=&FloatFromCurve(
     FloatCurve=DataInterface() {propertyValues {Curve {keys [
         {arriveTangent=5.4511, interpMode=RCIM_Cubic, leaveTangent=5.4511, tangentMode=RCTM_User, time=0, value=0.4}
         {arriveTangent=-2.5865, interpMode=RCIM_Cubic, leaveTangent=-2.5865, tangentMode=RCTM_User, time=0.2, value=1}
         {arriveTangent=-0.4609, interpMode=RCIM_Cubic, leaveTangent=-0.4609, tangentMode=RCTM_User, time=1, value=0}
        ]}}}
     Scale Curve=&RandomRangeFloat(Minimum=0.5, Maximum=0.3)
    )
    Scale RGB=&MakeVectorFromLinearColorRGB(Color=&ColorFromCurve(ColorCurve=DataInterface() {propertyValues {
        AlphaCurve {keys []}
        BlueCurve {keys [<at 0.0393 2.4 RCIM_Cubic>,<at 0.1425 0.1729>,<at 0.4414 0.08>]}
        GreenCurve {keys [<at 0.0393 2.7 RCIM_Cubic>,<at 0.1425 0.3797>,<at 0.4414 0.0883>]}
        RedCurve {keys [<at 0.0393 6 RCIM_Cubic>,<at 0.1425 1>,<at 0.4414 0.1>]}
       }}))
    ScaleA=true
    ScaleRGB=true
   )
   &DynamicMaterialParameters(
    Index 0 Param 1=&Add_Float(
     A=$Particles.RandomFloat
     B=&FloatFromCurve(
      FloatCurve=DataInterface() {propertyValues {Curve {keys [
          {arriveTangent=2.7533, interpMode=RCIM_Cubic, leaveTangent=1.8958, tangentMode=RCTM_User, time=0, value=1}
          {arriveTangent=0.8292, interpMode=RCIM_Cubic, leaveTangent=0.8292, tangentMode=RCTM_User, time=0.321, value=1.3453}
          {arriveTangent=0.0529, interpMode=RCIM_Cubic, leaveTangent=0.2021, tangentMode=RCTM_User, time=1, value=1.5}
         ]}}}
      CurveIndex=&Add_Float(A=&Multiply_Float(A=$Particles.Initial.MaterialRandom, B=0.5), B=$Particles.NormalizedAge)
     )
    )
    Index 0 Param 2=&FloatFromCurve(FloatCurve=DataInterface() {propertyValues {Curve {keys [
         {arriveTangent=0.0453, interpMode=RCIM_Cubic, tangentMode=RCTM_Break, time=0, value=-1}
         {arriveTangent=0.0604, interpMode=RCIM_Cubic, leaveTangent=0.0604, tangentMode=RCTM_User, time=1, value=-1}
        ]}}})
    Index 0 Param 3=&FloatFromCurve(FloatCurve=DataInterface() {propertyValues {Curve {keys [
         {arriveTangent=-53.9199, interpMode=RCIM_Cubic, leaveTangent=-8.9727, tangentMode=RCTM_User, time=0, value=1}
         {arriveTangent=0.2347, interpMode=RCIM_Cubic, leaveTangent=0.2723, tangentMode=RCTM_Break, time=0.25, value=0}
        ]}}})
    Index 0 Param 4=&FloatFromCurve(FloatCurve=DataInterface() {propertyValues {Curve {keys [<at 0.5 1>,<at 1 0>]}}})
    Index 2 Param 1=&FloatFromCurve(FloatCurve=DataInterface() {propertyValues {Curve {keys [
         {interpMode=RCIM_Cubic, leaveTangent=4.2485, tangentMode=RCTM_Break, time=0, value=0}
         {arriveTangent=1.0548, interpMode=RCIM_Cubic, leaveTangent=1.0548, tangentMode=RCTM_User, time=0.17, value=0.5715}
         <at 1 1>
        ]}}})
    Index 2 Param 2=&FloatFromCurve(
     FloatCurve=DataInterface() {propertyValues {Curve {keys [
         {arriveTangent=0.9316, interpMode=RCIM_Cubic, leaveTangent=0.9316, tangentMode=RCTM_User, time=0, value=0.1}
         {arriveTangent=0.0259, interpMode=RCIM_Cubic, leaveTangent=0.0259, tangentMode=RCTM_User, time=1, value=0.4}
        ]}}}
     Scale Curve=&RandomRangeFloat(Minimum=0.2, Maximum=0.7)
    )
    Index 2 Param 3=&FloatFromCurve(FloatCurve=DataInterface() {propertyValues {Curve {keys [
         {interpMode=RCIM_Cubic, leaveTangent=2.3064, tangentMode=RCTM_Break, time=0, value=1}
         {interpMode=RCIM_Cubic, leaveTangent=11.1233, tangentMode=RCTM_Break, time=1, value=2}
        ]}}})
    Index 2 Param 4=&FloatFromCurve(
     FloatCurve=DataInterface() {propertyValues {Curve {keys [
         {arriveTangent=-0.6867, interpMode=RCIM_Cubic, leaveTangent=-0.6867, tangentMode=RCTM_User, time=0, value=1}
         {interpMode=RCIM_Cubic, leaveTangent=-1.2956, tangentMode=RCTM_Break, time=1, value=0.75}
        ]}}}
     Scale Curve=1.6
    )
    Module.Param0WriteEnabled=true
    Module.Param0WriteEnabled2=true
    Module.Param1WriteEnabled=false
    Module.Param1WriteEnabled2=true
    Module.Param2WriteEnabled=true
    Module.Param2WriteEnabled2=true
    Module.Param3WriteEnabled=true
    Module.Param3WriteEnabled2=true
    Write Parameter Index 2=true
   )
   &CameraOffset(Camera Offset Amount=&CalculateParticleRadius(
     Radius Calculation Type:ENiagaraCollisionRadiusOptions=::NewEnumerator0
     Particle Radius:NiagaraFloat=1
     Mesh Dimensions:Vector3f=<1 1 1>
     Method for Calculating Particle Radius:ENiagaraCalculateRadiusOptions=::NewEnumerator2
    ))
  ]
  rendererClasses [&NiagaraSpriteRendererProperties]
  renderers [
   {rendererClass=&NiagaraSpriteRendererProperties}
   {rendererClass=&NiagaraSpriteRendererProperties, rendererIndex=1}
  ]
 }
 DustCircle {rendererClasses [], renderers []}
 EnergyCone {
  emitterUpdateScript [
   &EmitterState()
   &SpawnBurst_Instantaneous(Use Spawn Probability=false, UseLoopCountLimit=true)
  ]
  particleSpawnScript [
   &InitializeParticle(
    Color=<rgba 0.8 0 0.0136 1>
    Lifetime=2.5
    Mesh Scale=&Multiply_VectorByFloat(Vector=<0.6 1 1>, Float=2)
    Mesh Scale Mode=::NewEnumerator4
    Position Offset:NiagaraPosition=$System.ShipGroundOffset
    Sprite Size=&Multiply_Vector2D_ByFloat(Float=150, Vector2D=<1 1>)
    Sprite Size Mode=::NewEnumerator4
    UsePositionOffset=true
    Write Material Random=true
   )
   &InitialMeshOrientation(
    Mesh Axis To Orient=$System.WindDirection
    Mesh Orientation Mode=::NewEnumerator2
    Orientation Vector=<-1 0 0>
    Rotation=<0 0.5 0>
    Use Rotation Vector=false
   )
   &CameraOffset(Camera Offset Amount=5)
   /SetVariables(Particles.CameraOffsetFlare=15) {isSetParameters=true}
  ]
  particleUpdateScript [
   &ParticleState()
   &ScaleMeshSize(Scale Factor=&Multiply_VectorByFloat(
     Vector=&MakeVector(
      X=&FloatFromCurve(FloatCurve=DataInterface() {propertyValues {Curve {keys [
           {arriveTangent=0.0679, interpMode=RCIM_Cubic, leaveTangent=0.0679, tangentMode=RCTM_User, time=0, value=0.1}
           {interpMode=RCIM_Cubic, tangentMode=RCTM_User, time=1, value=1}
          ]}}})
      Y=1
      Z=1
     )
     Float=&MakeFloatFromVector(
      Vector=&VectorFromFloat(Value=&Lerp_Float(
        A=1
        B=2.5
        Alpha=&RemapRange(
         Input Value To Remap=/Niagara/DynamicInputs/Camera/DistanceToCamera(Use Owner Position:NiagaraBool=true, Camera Query:NiagaraDataInterfaceCamera=DataInterface())
         Input Min=12500
         Input Max=25000
         Output Min=0
         Output Max=1
        )
       ))
      Channel=::NewEnumerator0
     )
    )) {enabled=false}
   &ScaleColor(
    Scale Alpha=&FloatFromCurve(FloatCurve=DataInterface() {propertyValues {Curve {keys [
         {arriveTangent=5.5602, interpMode=RCIM_Cubic, leaveTangent=5.5602, tangentMode=RCTM_User, time=0, value=0}
         {interpMode=RCIM_Cubic, tangentMode=RCTM_User, time=0.3, value=1}
         {arriveTangentWeight=0.1026, interpMode=RCIM_Cubic, leaveTangentWeight=0.1368, tangentMode=RCTM_User, time=0.6, value=1}
         {arriveTangent=-0.4064, interpMode=RCIM_Cubic, leaveTangent=-0.4064, tangentMode=RCTM_User, time=1, value=0}
        ]}}})
    ScaleA=true
    ScaleRGB=false
   ) {enabled=false}
   /SetVariables(
    Particles.ColorGlow:LinearColor=&ColorFromCurve(ColorCurve=DataInterface() {propertyValues {
       AlphaCurve {keys [<at 0.012 0>,<at 0.0775 1 RCIM_Cubic>,<at 0.8636 0.8>,<at 1 0 RCIM_Cubic>]}
       BlueCurve {keys [<at 0.0829 0.4057>,<at 0.2817 0.1656>,<at 0.8924 0.1052>]}
       GreenCurve {keys [<at 0.0829 0.4>,<at 0.2817 0.1599>,<at 0.8924 0.09>]}
       RedCurve {keys [<at 0.0829 8>,<at 0.2817 3>,<at 0.8924 1>]}
      }})
    Particles.PositionGlow:NiagaraPosition=&AddVectorToPosition(Position=$Particles.Position, Vector=<0 0 0>)
   ) {enabled=false, isSetParameters=true}
   /SetVariables(
    Particles.ColorGlowSoft:LinearColor=&MultiplyLinearColor(Linear Color=$Particles.ColorGlow, Scale Factor=<1 0 0 1>)
    Particles.SpriteSizeGlow:Vector2f=&Multiply_Vector2D_ByFloat(Float=0.6, Vector2D=$Particles.SpriteSize)
   ) {enabled=false, isSetParameters=true}
   /SetVariables(Particles.SpriteSizeLong:Vector2f=&Multiply_Vector2D(A=$Particles.SpriteSize, B=<2 0.1>)) {enabled=false, isSetParameters=true}
   /SetVariables(Particles.ColorMain:LinearColor=$Particles.Color, Particles.ScaleMain:Vector3f=$Particles.Scale) {enabled=false, isSetParameters=true}
   /SetVariables(
    Particles.ColorCore:LinearColor=&MultiplyLinearColor(
     Linear Color=&ColorFromCurve(ColorCurve=DataInterface() {propertyValues {
        AlphaCurve {keys [<at 0 1 RCIM_Cubic>]}
        BlueCurve {keys [
          <at 0.2202 2 RCIM_Cubic>
          <at 0.2442 0.5>
          <at 0.3077 0.1>
          {interpMode=RCIM_Cubic, leaveTangent=0.3683, time=0.4543, value=0.0393}
          <at 0.6847 0.24>
         ]}
        GreenCurve {keys [
          <at 0.2202 0.95 RCIM_Cubic>
          <at 0.2442 0.75>
          <at 0.3077 0.25>
          {interpMode=RCIM_Cubic, leaveTangent=2.3892, time=0.4543, value=0.3589}
          <at 0.6847 1.1583>
         ]}
        RedCurve {keys [<at 0.2202 0.2 RCIM_Cubic>, <at 0.2442 1>, <at 0.3077 1>, {interpMode=RCIM_Cubic, leaveTangent=5.261, time=0.4543, value=1}, <at 0.6847 3>]}
       }})
     Scale Factor=&MakeVector4FromVectorAndFloat(
      "Vector (RGB)"=<1 1 1>
      "Float (Alpha)"=&FloatFromCurve(FloatCurve=DataInterface() {propertyValues {Curve {keys [
           {interpMode=RCIM_Cubic, tangentMode=RCTM_User, time=0, value=0}
           {interpMode=RCIM_Cubic, tangentMode=RCTM_User, time=0.2, value=1}
           {interpMode=RCIM_Cubic, tangentMode=RCTM_User, time=0.75, value=1}
           {interpMode=RCIM_Cubic, tangentMode=RCTM_User, time=1, value=0}
          ]}}})
     )
    )
    Particles.ScaleCore:Vector3f=&Multiply_Vector(A=$Particles.Scale, B=<1 1 1>)
   ) {isSetParameters=true}
   /SetVariables(
    Particles.ColorDark:LinearColor=&MultiplyLinearColor(
     Linear Color=<rgba 0.02 0.018 0.0183 1.6>
     Scale Factor=&MakeVector4FromVectorAndFloat(
      "Vector (RGB)"=<1 1 1>
      "Float (Alpha)"=&FloatFromCurve(FloatCurve=DataInterface() {propertyValues {Curve {keys [<at 0 0>,<at 0.5 1>,<at 0.692 0.811>,<at 1 0>]}}})
     )
    )
    Particles.ScaleDark:Vector3f=&Multiply_VectorByFloat(Vector=$Particles.Scale, Float=0.6)
   ) {isSetParameters=true}
   /SetVariables(Particles.SpriteSizeDark:Vector2f=&Multiply_Vector2D_ByFloat(Float=1, Vector2D=$Particles.SpriteSize)) {enabled=false, isSetParameters=true}
  ]
  rendererClasses [&NiagaraMeshRendererProperties]
  renderers [
   {rendererClass=&NiagaraMeshRendererProperties, rendererIndex=0}
   {rendererClass=&NiagaraMeshRendererProperties, rendererIndex=1}
   {rendererClass=&NiagaraMeshRendererProperties, rendererIndex=2}
  ]
 }
 Minimal {
  bEnabled=false
  emitterUpdateScript [
   &EmitterState(
    Life Cycle Mode=::NewEnumerator1
    Loop Behavior=::NewEnumerator1
    Loop Delay=&Add_Float(A=$System.ChargeUpTime, B=-0.1)
    Loop Duration=1
    UseLoopDelay=true
   )
   &SpawnBurst_Instantaneous(Use Spawn Probability=false, UseLoopCountLimit=false)
  ]
  particleSpawnScript [&InitializeParticle(
    Color=<rgba 3 0.0382 0.0382 1>
    Lifetime=0.3
    Sprite Size=<333 333>
    Sprite Size Mode=::NewEnumerator4
    UsePositionOffset=false
    Write Material Random=true
   )]
  particleUpdateScript [
   &ParticleState()
   &ScaleSpriteSize(Uniform Curve Sprite Scale=DataInterface() {propertyValues {Curve {keys [
        {arriveTangent=-1.9135, interpMode=RCIM_Cubic, leaveTangent=-1.9135, tangentMode=RCTM_User, time=0, value=1}
        {interpMode=RCIM_Cubic, tangentMode=RCTM_User, time=1, value=0}
       ]}}})
   &ScaleColor(ScaleA=true, ScaleRGB=true) {enabled=false}
  ]
  rendererClasses [&NiagaraSpriteRendererProperties]
  renderers [{rendererClass=&NiagaraSpriteRendererProperties}]
 }
 Wind {
  emitterUpdateScript [
   &EmitterState(
    Enable Distance Culling=true
    Life Cycle Mode=::NewEnumerator1
    Loop Behavior=::NewEnumerator1
    Loop Delay=0.5
    Loop Duration=2
    MaxDistance=6000
    Module.bUseMaxDistance=true
    Module.bUseMinDistance=false
    Scalability Mode=::NewEnumerator1
    UseLoopDelay=true
   )
   &SpawnRate(SpawnRate=12, Use Spawn Probability=false)
  ]
  particleSpawnScript [
   /SetVariables(Particles.WindScale:NiagaraFloat=&RandomRangeFloat(Minimum=0, Maximum=1)) {isSetParameters=true}
   &InitializeParticle(
    Color=<rgba 0.618 0.7516 0.8 0.6>
    Lifetime=&RandomRangeFloat(Minimum=0.5, Maximum=0.6)
    Mesh Scale=&Lerp_Vector(A=<2 2 3>, B=<3 3 5>, Alpha=$Particles.WindScale)
    Mesh Scale Mode=::NewEnumerator4
    Position Offset=&AddVectorToPosition(Position=$System.ShipGroundOffset, Vector=&Multiply_VectorByFloat(Vector=$System.WindDirection, Float=-150))
    UsePositionOffset=true
    Write Material Random=true
   )
   &InitialMeshOrientation(
    Mesh Axis To Orient=<0 0 1>
    Mesh Orientation Mode=::NewEnumerator2
    Orientation Coordinate Space=::Local
    Orientation Vector=$System.WindDirection
    Rotation=&RandomRangeVector(Minimum=<0 0 0>, Maximum=<0 0 1>)
    Use Rotation Vector=true
   )
   &AddVelocity(Velocity=$System.WindDirection, Velocity Speed Scale=-700)
  ]
  particleUpdateScript [
   &ParticleState()
   /SetVariables(Particles.Position=&Add_Vector(
     A=Linked() {linkedVariable {name=Particles.Initial.Position, type=NiagaraPosition}}
     B=&Multiply_VectorByFloat(
      Vector=&ReturnMeshOrientationAxis(Mesh Orientation Axis=::ZAxis, Mesh Orientation Quat=$Particles.MeshOrientation)
      Float=&FloatFromCurve(
       FloatCurve=DataInterface() {propertyValues {Curve {keys [{arriveTangent=-2.1833, interpMode=RCIM_Cubic, leaveTangent=-2.1833, tangentMode=RCTM_User, time=0, value=0}, <at 1 -1>]}}}
       Scale Curve=&Lerp_Float(A=300, B=600, Alpha=$Particles.WindScale)
      )
     )
    )) {enabled=false, isSetParameters=true}
   &UpdateMeshOrientation(
    Rotation Coordinate Space=::NewEnumerator3
    Rotation Rate=&FloatFromCurve(FloatCurve=DataInterface() {propertyValues {Curve {keys [
         {arriveTangent=-2.0408, interpMode=RCIM_Cubic, time=0, value=1}
         {arriveTangent=-2.0603, interpMode=RCIM_Cubic, leaveTangent=-11.7273, tangentMode=RCTM_User, time=1, value=0.05}
        ]}}})
    Rotation Vector=<0 0 1>
   ) {enabled=false, module=UpdateMeshOrientation001}
   /SetVariables(
    Particles.WindLength:NiagaraFloat=&FloatFromCurve(
     FloatCurve=DataInterface() {propertyValues {Curve {keys [
         {arriveTangent=0.9197, interpMode=RCIM_Cubic, leaveTangent=0.9197, tangentMode=RCTM_User, time=0, value=0.2}
         {interpMode=RCIM_Cubic, tangentMode=RCTM_User, time=1, value=1}
        ]}}}
     Scale Curve=4
    )
    Particles.WindWidth:NiagaraFloat=&FloatFromCurve(
     FloatCurve=DataInterface() {propertyValues {Curve {keys [
         {arriveTangent=0.9993, interpMode=RCIM_Cubic, leaveTangent=0.9993, tangentMode=RCTM_User, time=0, value=0.6}
         {interpMode=RCIM_Cubic, tangentMode=RCTM_User, time=1, value=1}
        ]}}}
     Scale Curve=1.8
    )
   ) {isSetParameters=true}
   &ScaleMeshSize(Scale Factor=&MakeVector(X=$Particles.WindWidth, Y=$Particles.WindWidth, Z=$Particles.WindLength))
   &CameraOffset(Camera Offset Amount=-30)
   &ScaleColor(
    Scale Alpha=&FloatFromCurve(FloatCurve=DataInterface() {propertyValues {Curve {keys [
         {arriveTangent=8.125, interpMode=RCIM_Cubic, time=0, value=0}
         {arriveTangent=-0.4175, interpMode=RCIM_Cubic, leaveTangent=-0.4175, tangentMode=RCTM_User, time=0.3, value=1}
         {arriveTangent=-1.8598, interpMode=RCIM_Cubic, leaveTangent=-1.8598, tangentMode=RCTM_User, time=1, value=0}
        ]}}})
    Scale RGB=&MakeVectorFromLinearColorRGB(Color=&ColorFromCurve(ColorCurve=DataInterface() {propertyValues {
        AlphaCurve {keys []}
        BlueCurve {keys [<at 0.0284 0.06 RCIM_Cubic>,<at 0.3496 0.1>,<at 0.9591 0.3 RCIM_Cubic>]}
        GreenCurve {keys [<at 0.0284 0.697 RCIM_Cubic>,<at 0.3496 0.19>,<at 0.9591 0.3 RCIM_Cubic>]}
        RedCurve {keys [<at 0.0284 3 RCIM_Cubic>,<at 0.3496 1>,<at 0.9591 0.3 RCIM_Cubic>]}
       }}))
    ScaleA=true
    ScaleRGB=true
   )
   &DynamicMaterialParameters(
    Index 0 Param 1=&FloatFromCurve(FloatCurve=DataInterface() {propertyValues {Curve {keys [
         {interpMode=RCIM_Cubic, tangentMode=RCTM_User, time=0, value=0}
         {arriveTangent=1.9778, interpMode=RCIM_Cubic, leaveTangent=1.9778, tangentMode=RCTM_User, time=1, value=1}
        ]}}})
    Index 0 Param 2=&FloatFromCurve(FloatCurve=DataInterface() {propertyValues {Curve {keys [<at 0 0>,<at 1 1>]}}})
    Index 0 Param 3=1
    Module.Param0WriteEnabled=true
    Module.Param1WriteEnabled=true
    Module.Param2WriteEnabled=true
    Module.Param3WriteEnabled=false
   )
   &Drag(Drag=0.5, Use Linear Drag=true, Use Rotational Drag=true)
   &SolveForcesAndVelocity(Clamp Velocity=false, Limit Acceleration=false)
  ]
  rendererClasses [&NiagaraMeshRendererProperties]
  renderers [{rendererClass=&NiagaraMeshRendererProperties, rendererIndex=0}]
 }
 WindStreaks {
  emitterUpdateScript [
   &EmitterState(Life Cycle Mode=::NewEnumerator1, Loop Behavior=::NewEnumerator1, Loop Duration=1.5, UseLoopDelay=true)
   &SpawnRate(
    SpawnRate=&FloatFromCurve(
     FloatCurve=DataInterface() {propertyValues {Curve {keys [{interpMode=RCIM_Cubic, tangentMode=RCTM_User, time=0.3, value=0.2}, {interpMode=RCIM_Cubic, tangentMode=RCTM_User, time=1.1, value=1}]}}}
     CurveIndex=$Emitter.Age
     Scale Curve=60
    )
    Use Spawn Probability=false
   )
  ]
  particleSpawnScript [
   &InitializeParticle(
    Color Maximum=<rgba 0.5521 0.4601 0.4417 0.4>
    Color Minimum=<rgba 0.4844 0.4359 0.4359 0.5>
    Color Mode=::NewEnumerator2
    Lifetime=&Multiply_Float(
     A=&FloatFromCurve(
      FloatCurve=DataInterface() {propertyValues {Curve {keys [{interpMode=RCIM_Cubic, tangentMode=RCTM_User, time=0, value=2}, {interpMode=RCIM_Cubic, tangentMode=RCTM_User, time=1.5, value=1}]}}}
      CurveIndex=$Emitter.Age
     )
     B=&RandomRangeFloat(Minimum=0.3, Maximum=0.4)
    )
    Sprite Size=&Multiply_Vector2D_ByFloat(Float=200, Vector2D=&RandomRangeVector2D(Minimum=<0.6 2>, Maximum=<1 3>))
    Sprite Size Mode=::NewEnumerator4
    UsePositionOffset=false
    Write Material Random=true
   )
   &PartitionParticles()
   &ShapeLocation(
    Hemisphere Distribution=<1 0.5>
    MaskSpawnGroup=false
    Non Uniform Scale=<3 1 0.5>
    Offset=&RotateVectorByQuaternion(
     Vector To Rotate=&AddVectorToPosition(
      Position=$System.DustGroundOffset
      Vector=&MergeHelper_VectorSwitch(0=<-50 0 0>, 1=<50 0 0>, 2=?, 3=?, Index=$Particles.PartitionParticles.Partition, Num=::NewEnumerator0)
     )
     Quaternion=$Engine.Owner.Rotation
    )
    Offset Mode=::NewEnumerator0
    Override Seed=false
    Sphere Radius=20
    UseOwnerScale=true
   )
   &AddVelocity(
    Module.Use Falloff Distance=false
    Origin Offset=&AddVectorToPosition(
     Position=$System.DustGroundOffset
     Vector=&AddVectorToPosition(Position=<0 0 -30>, Vector=&Multiply_VectorByFloat(Vector=$System.WindDirection, Float=100))
    )
    Override Seed=false
    Velocity Mode=::NewEnumerator1
    Velocity Speed=&RandomRangeFloat(Minimum=500, Maximum=700)
   )
   &SubUVAnimation(
    "Sprite Renderer "=DataInterface() {propertyValues {SpriteRenderer=&self:WindStreaks_0.NiagaraSpriteRendererProperties_1}}
    SubUV Animation Mode:ENiagara_SubUVLookupModeV2=::NewEnumerator2
    UseEndFrame=false
    UseStartFrame=false
   )
   &CameraOffset(Camera Offset Amount=50)
  ]
  particleUpdateScript [
   &ParticleState()
   &ScaleSpriteSize(
    Scale Factor=&MakeVector2D(
     X=&FloatFromCurve(FloatCurve=DataInterface() {propertyValues {Curve {keys [<at 0 0.4>,<at 0.4 1>]}}})
     Y=&FloatFromCurve(FloatCurve=DataInterface() {propertyValues {Curve {keys [
          {arriveTangent=0.2172, interpMode=RCIM_Cubic, leaveTangent=0.2172, tangentMode=RCTM_User, time=0, value=0.8}
          {interpMode=RCIM_Cubic, tangentMode=RCTM_User, time=1, value=1}
         ]}}})
    )
    Scale Sprite Size Mode=::NewEnumerator2
   )
   &Drag(Drag=4, Use Linear Drag=true, Use Rotational Drag=true)
   &InheritSourceMovement(
    Applied Position Scale=&VectorFromFloat(Value=&FloatFromCurve(FloatCurve=DataInterface() {propertyValues {Curve {keys [<at 0 1>,<at 0.5 0>]}}}, Scale Curve=0.4))
    ApplyForce=false
    ApplyPosition=true
    ApplyPositionFalloff=false
   )
   &SolveForcesAndVelocity(Clamp Velocity=false, Limit Acceleration=false)
   &ScaleColor(
    Scale Alpha=&FloatFromCurve(FloatCurve=DataInterface() {propertyValues {Curve {keys [<at 0 1>,<at 1 0>]}}}, Scale Curve=5)
    Scale RGB=&VectorFromFloat(Value=0.55)
    ScaleA=true
    ScaleRGB=true
   )
   &DynamicMaterialParameters(
    Index 0 Param 1=20
    Index 0 Param 2=1
    Index 0 Param 3=1
    Index 0 Param 4=&FloatFromCurve(FloatCurve=DataInterface() {propertyValues {Curve {keys [
         {interpMode=RCIM_Cubic, tangentMode=RCTM_User, time=0, value=0}
         {arriveTangent=0.9103, interpMode=RCIM_Cubic, leaveTangent=0.9103, tangentMode=RCTM_User, time=0.8, value=1}
        ]}}})
    Index 2 Param 1=2
    Index 2 Param 2=&FloatFromCurve(FloatCurve=DataInterface() {propertyValues {Curve {keys [
         {arriveTangent=0.0308, interpMode=RCIM_Cubic, leaveTangent=0.0308, tangentMode=RCTM_User, time=0, value=1}
         {arriveTangent=-2.7663, interpMode=RCIM_Cubic, leaveTangent=-2.7663, tangentMode=RCTM_User, time=1, value=0}
        ]}}})
    Module.Param0WriteEnabled=true
    Module.Param0WriteEnabled2=true
    Module.Param1WriteEnabled=false
    Module.Param1WriteEnabled2=true
    Module.Param2WriteEnabled=false
    Module.Param2WriteEnabled2=true
    Module.Param3WriteEnabled=true
    Module.Param3WriteEnabled2=true
    Write Parameter Index 2=true
   )
  ]
  rendererClasses [&NiagaraSpriteRendererProperties]
  renderers [{rendererClass=&NiagaraSpriteRendererProperties}]
 }
}
summary {
 systemName=NS_SouthBulk_ChargeStart
 emitters [
  {emitterName=Wind, rendererClasses [&NiagaraMeshRendererProperties]}
  {bEnabled=false, emitterName=Dust, rendererClasses [&NiagaraSpriteRendererProperties]}
  {bEnabled=false, emitterName=Minimal, rendererClasses [&NiagaraSpriteRendererProperties]}
  {bEnabled=false, emitterName=DirtChunks, rendererClasses [&NiagaraSpriteRendererProperties]}
  {bEnabled=false, emitterName=Dust001, rendererClasses [&NiagaraSpriteRendererProperties]}
  {emitterName=WindStreaks, rendererClasses [&NiagaraSpriteRendererProperties]}
  {emitterName=EnergyCone, rendererClasses [&NiagaraMeshRendererProperties]}
  {emitterName=DustCircle, rendererClasses []}
 ]
 userVariables [
  User.IsSliding {type=NiagaraBool, defaultValue=true}
  User.NormalImpact {type=Vector3f, description="", defaultValue {struct=Vector3f, value=<0 0 0>}}
 ]
}
